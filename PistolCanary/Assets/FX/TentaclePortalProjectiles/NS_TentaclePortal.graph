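# Niagara system for the purple tentacle portal effect: a timed portal that pops open, holds for a few seconds and collapses.
# One Portal particle drives sprite, decal and mesh renderers through separate curve-animated size attributes, backed by a glow sprite and light.
# Staggered mesh emitters add spinning energy, line, smoke and trail rings, some flashing bright by emitter age or counter-rotating by partition.
# Front-end and time-of-day inputs adjust some scale and brightness values.
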NiagaraSystem /PistolCanary/Assets/FX/TentaclePortalProjectiles/NS_TentaclePortal
dependencies {
 usedDataInterfaces [
  &NiagaraDataInterfaceArrayColor
  &NiagaraDataInterfaceCurve
  &NiagaraDataInterfaceVector2DCurve
  &NiagaraDataInterfaceSpriteRendererInfo
  &NiagaraDataInterfaceArrayFloat
  &NiagaraDataInterfaceColorCurve
  &NiagaraDataInterfaceVectorCurve
 ]
 usedDynamicInputs [&Vector2DFromCurve, &RandomRangeVector2D]
 usedModules [
  &self:SystemSpawnScript.SystemScriptSource.SystemScriptGraph.NiagaraNodeAssignment_1.SetVariables_#1
  &SystemState
  &Timeline
  &self:SystemSpawnScript.SystemScriptSource.SystemScriptGraph.NiagaraNodeAssignment_0.SetVariables_#2
  &self:SystemSpawnScript.SystemScriptSource.SystemScriptGraph.NiagaraNodeAssignment_2.SetVariables_#3
  &self:SystemSpawnScript.SystemScriptSource.SystemScriptGraph.NiagaraNodeAssignment_3.SetVariables_#4
 ]
}
emitters {
 Glow {
  emitterUpdateScript [
   &EmitterState(Life Cycle Mode=::NewEnumerator1, Loop Behavior=::NewEnumerator1, Loop Delay=0.1, Loop Duration=3.5, UseLoopDelay=true)
   &SpawnBurst_Instantaneous(Use Spawn Probability=false, UseLoopCountLimit=false)
  ]
  particleSpawnScript [
   &InitializeParticle(
    Color=<rgba 0.163 0.0234 0.3791 0.7507>
    Lifetime=2.7
    Sprite Size Mode=::NewEnumerator3
    Uniform Sprite Size=2000
    UsePositionOffset=false
    Write Material Random=true
   )
   &CameraOffset(Camera Offset Amount=30)
   &SpriteFacingAndAlignment(Sprite Facing=<0 0 1>, WriteAlignment=false, WriteFacing=true)
  ]
  particleUpdateScript [
   &ParticleState()
   &FN_Light_Attributes(
    Color=<rgba 1.4181 0 5 1>
    Create Color Attribute=true
    Create Exponent Attribute=true
    Create Position Attribute=true
    Create Radius Attribute=true
    Enable Particle Light Enable=true
    Exponent=2
    Light Enabled=true
    Radius=250
    Scale Color=&FloatFromCurve(
     FloatCurve=DataInterface() {propertyValues {Curve {keys [
         <at 0 0 RCIM_Cubic>
         {arriveTangentWeight=0.0581, interpMode=RCIM_Cubic, leaveTangentWeight=0.349, tangentMode=RCTM_User, time=0.0705, value=1}
         {arriveTangent=-1.0758, interpMode=RCIM_Cubic, leaveTangent=-1.0758, time=0.9036, value=0.8288}
         {arriveTangent=-3.8525, interpMode=RCIM_Cubic, leaveTangent=-3.8525, tangentMode=RCTM_User, time=1, value=0}
        ]}}}
     Scale Curve=&FN_FrontEndFloatModifier(Gameplay=4, Front End=1)
    )
   )
   &ScaleColor(
    Scale Alpha=&FloatFromCurve(FloatCurve=DataInterface() {propertyValues {Curve {keys [
         {arriveTangent=5.0727, interpMode=RCIM_Cubic, leaveTangent=5.0727, tangentMode=RCTM_User, time=0, value=0.1641}
         {
          arriveTangent=0.1472
          arriveTangentWeight=0.3592
          interpMode=RCIM_Cubic
          leaveTangent=0.1472
          leaveTangentWeight=0.3592
          tangentMode=RCTM_User
          tangentWeightMode=RCTWM_WeightedBoth
          time=0.5103
          value=1
         }
         {tangentMode=RCTM_User, time=1, value=0}
        ]}}})
    ScaleA=true
    ScaleRGB=true
   )
  ]
  rendererClasses [&NiagaraSpriteRendererProperties, &NiagaraLightRendererProperties]
  renderers [
   {rendererClass=&NiagaraSpriteRendererProperties}
   {rendererClass=&NiagaraLightRendererProperties, rendererIndex=1}
  ]
 }
 OpenPortal {
  emitterUpdateScript [
   &EmitterState(Life Cycle Mode=::NewEnumerator1, Loop Behavior=::NewEnumerator1, Loop Delay=0.05, Loop Duration=2, UseLoopDelay=true)
   &SpawnBurst_Instantaneous(Spawn Count=4, Use Spawn Probability=false, UseLoopCountLimit=false)
  ]
  particleSpawnScript [
   &InitializeParticle(
    Lifetime Max=0.5
    Lifetime Min=0.8
    Lifetime Mode=::NewEnumerator1
    Mesh Scale=&Multiply_VectorByFloat(
     Vector=&RandomRangeVector(Minimum=<1 1 1.5>, Maximum=<1 1 0.5>)
     Float=&RandomRangeFloat(Minimum=2, Maximum=3)
    )
    Mesh Scale Mode=::NewEnumerator4
    UsePositionOffset=false
    Write Material Random=true
   )
   &InitialMeshOrientation(
    Mesh Orientation Mode=::NewEnumerator2
    Rotation=&RandomRangeVector(Minimum=<0.03 0.03 0>, Maximum=<0.03 0.03 0>)
    Use Rotation Vector=true
   )
   &ShapeLocation(
    Cylinder Height=10
    Cylinder Radius=1
    Hemicircle Y=true
    MaskSpawnGroup=false
    Offset=&RandomRangeVector(Minimum=<0 0 10>, Maximum=<0 0 1>)
    Offset Mode=::NewEnumerator0
    Override Seed=false
    Shape Primitive=::NewEnumerator1
    Surface Only=false
    UseOwnerScale=true
   )
   &AddVelocity(Velocity=&RandomRangeVector(Minimum=<0 0 0>, Maximum=<0 0 50>))
   &PartitionParticles()
  ]
  particleUpdateScript [
   &ParticleState()
   &DynamicMaterialParameters(
    Index 0 Param 1=&FloatFromCurve(FloatCurve=DataInterface() {propertyValues {Curve {keys [{interpMode=RCIM_Cubic, tangentMode=RCTM_User, time=0, value=0.8}, {interpMode=RCIM_Cubic, tangentMode=RCTM_User, time=1, value=0}]}}})
    Module.Param0WriteEnabled=true
    Module.Param1WriteEnabled=true
    Module.Param2WriteEnabled=true
    Module.Param3WriteEnabled=true
   )
   &ScaleMeshSize(Scale Factor=&VectorFromCurve(
     VectorCurve=DataInterface() {propertyValues {
       XCurve {keys [
         {arriveTangent=0.0402, interpMode=RCIM_Cubic, leaveTangent=0.0402, tangentMode=RCTM_User, time=-0.0344, value=0.4}
         {arriveTangent=0.8129, interpMode=RCIM_Cubic, leaveTangent=0.8129, time=0.3105, value=0.7736}
         <at 0.7037 1 RCIM_Cubic>
        ]}
       YCurve {keys [
         {arriveTangent=0.1314, interpMode=RCIM_Cubic, leaveTangent=0.1314, tangentMode=RCTM_User, time=-0.0344, value=0.4}
         {arriveTangent=0.7059, interpMode=RCIM_Cubic, leaveTangent=0.7059, time=0.1877, value=0.8637}
         {arriveTangent=-0.0356, interpMode=RCIM_Cubic, leaveTangent=-0.0356, tangentMode=RCTM_User, time=0.8156, value=1}
        ]}
       ZCurve {keys [<at -0.0344 1>,<at 0.8156 1>]}
      }}
     Scale Curve=&Multiply_VectorByFloat(
      Vector=&Multiply_VectorByFloat(Vector=$Engine.Owner.Scale, Float=0.9)
      Float=&FloatFromCurve(
       FloatCurve=DataInterface() {propertyValues {Curve {keys [
           {arriveTangent=4.1785, interpMode=RCIM_Cubic, leaveTangent=4.1785, tangentMode=RCTM_User, time=0, value=0}
           {arriveTangent=1.1179, interpMode=RCIM_Cubic, leaveTangent=1.1179, tangentMode=RCTM_User, time=0.032, value=0.71}
           {arriveTangent=0.0224, interpMode=RCIM_Cubic, leaveTangent=0.0224, tangentMode=RCTM_User, time=1, value=1}
          ]}}}
       Scale Curve=4
      )
     )
    ))
   &ScaleColor(
    Scale RGB=&MakeVectorFromLinearColorRGB(Color=&LerpLinearColors(
      StartColor=&ColorFromCurve(ColorCurve=DataInterface() {propertyValues {
         AlphaCurve {keys []}
         BlueCurve {keys [
           {arriveTangent=0.3322, interpMode=RCIM_Cubic, leaveTangent=0.3322, tangentMode=RCTM_User, time=0.1294, value=0.9578}
           <at 0.4879 0.0133>
           <at 0.7343 0.0503>
          ]}
         GreenCurve {keys [
           {arriveTangent=-1.1199, interpMode=RCIM_Cubic, leaveTangent=-1.1199, tangentMode=RCTM_User, time=0.1294, value=0.1424}
           <at 0.4879 0.0039>
           <at 0.7343 0.0258>
          ]}
         RedCurve {keys [
           {arriveTangent=-3.8908, interpMode=RCIM_Cubic, leaveTangent=-3.8908, tangentMode=RCTM_User, time=0.1294, value=0.5264}
           <at 0.4879 0.0076>
           <at 0.7343 0.0165>
          ]}
        }})
      EndColor=<rgba 8.1359 1.0155 20 1>
      LerpFactor=&FloatFromCurve(
       FloatCurve=DataInterface() {propertyValues {Curve {keys [
           {arriveTangent=-10.5648, interpMode=RCIM_Cubic, leaveTangent=-10.5648, tangentMode=RCTM_User, time=0.1061, value=1}
           {interpMode=RCIM_Cubic, tangentMode=RCTM_User, time=0.2069, value=0.0897}
          ]}}}
       CurveIndex=$Emitter.Age
      )
     ))
    ScaleA=true
    ScaleRGB=true
   )
   &SolveForcesAndVelocity(Clamp Velocity=false, Limit Acceleration=false)
   /SetVariables(Particles.MeshOrientation=/Niagara/DynamicInputs/Helpers/MaskQuaternion(
     Mask Parameter:NiagaraInt32=$Particles.PartitionParticles.Partition
     Mask:NiagaraInt32=0
     Masked Quaternion:Quat4f=&MakeQuatFromAxisAngle(
      Axis=<0 0 1>
      AngleInDegrees=&FloatFromCurve(
       FloatCurve=DataInterface() {propertyValues {Curve {keys [
           {arriveTangent=-0.0197, interpMode=RCIM_Cubic, leaveTangent=-0.0197, tangentMode=RCTM_User, time=0.1034, value=0.5}
           {arriveTangent=-1.673, interpMode=RCIM_Cubic, leaveTangent=-1.673, time=0.505, value=0.204}
           {interpMode=RCIM_Cubic, leaveTangent=-0.7143, time=1, value=-1}
          ]}}}
       Scale Curve=200
      )
     )
     Passthrough Quaternion:Quat4f=&MakeQuatFromAxisAngle(
      Axis=<0 0 1>
      AngleInDegrees=&FloatFromCurve(
       FloatCurve=DataInterface() {propertyValues {Curve {keys [
           {interpMode=RCIM_Cubic, leaveTangentWeight=0.3518, tangentMode=RCTM_User, time=0.1034, value=0.5}
           {arriveTangent=1.673, interpMode=RCIM_Cubic, leaveTangent=1.673, time=0.523, value=0.827}
           <at 1 2 RCIM_Cubic>
          ]}}}
       Scale Curve=200
      )
     )
    )) {isSetParameters=true}
  ]
  rendererClasses [&NiagaraMeshRendererProperties]
  renderers [{rendererClass=&NiagaraMeshRendererProperties, rendererIndex=0}]
 }
 Portal {
  emitterUpdateScript [
   &EmitterState(
    Inactive Response=::NewEnumerator1
    Life Cycle Mode=::NewEnumerator1
    Loop Behavior=::NewEnumerator1
    Loop Delay=0.15
    Loop Duration=2.55
    UseLoopDelay=true
   )
   &SpawnBurst_Instantaneous(Use Spawn Probability=false, UseLoopCountLimit=false)
  ]
  particleSpawnScript [
   &InitializeParticle(
    Color=<rgba 3 3 3 2>
    Lifetime=2.55
    Mesh Scale Mode=::NewEnumerator3
    Mesh Uniform Scale=12
    Position Offset=<0 0 5>
    Sprite Rotation Angle=90
    Sprite Rotation Mode=::NewEnumerator0
    Sprite Size Mode=::NewEnumerator3
    Sprite UV Mode=::NewEnumerator4
    Uniform Sprite Size=750
    UsePositionOffset=true
    Write Material Random=true
   )
   &SpriteFacingAndAlignment(Sprite Facing=<0 0 1>, WriteAlignment=false, WriteFacing=true)
   &CameraOffset(Camera Offset Amount=10)
  ]
  particleUpdateScript [
   &ParticleState() {module=ParticleState001}
   &ScaleSpriteSize(
    Uniform Curve Scale=1.1
    Uniform Curve Sprite Scale=DataInterface() {propertyValues {Curve {keys [
        {interpMode=RCIM_Cubic, leaveTangentWeight=0.5835, tangentMode=RCTM_User, time=0, value=0.4}
        {arriveTangentWeight=0.0191, interpMode=RCIM_Cubic, tangentMode=RCTM_User, time=0.0573, value=1.2881}
        {
         arriveTangent=-0.6122
         arriveTangentWeight=0.0142
         interpMode=RCIM_Cubic
         leaveTangent=-0.6122
         tangentMode=RCTM_User
         time=0.1
         value=1.0358
        }
        {
         arriveTangent=0.194
         arriveTangentWeight=0.1175
         interpMode=RCIM_Cubic
         leaveTangent=0.194
         leaveTangentWeight=0.1586
         tangentMode=RCTM_User
         time=0.4526
         value=1.0792
        }
        {
         arriveTangent=-0.304
         arriveTangentWeight=0.0816
         interpMode=RCIM_Cubic
         leaveTangent=-0.304
         leaveTangentWeight=0.0611
         tangentMode=RCTM_User
         time=0.6975
         value=1.0877
        }
        {
         arriveTangent=-1.4355
         arriveTangentWeight=0.1427
         interpMode=RCIM_Cubic
         leaveTangent=-1.4355
         leaveTangentWeight=0.0397
         tangentMode=RCTM_User
         time=0.8808
         value=0.9662
        }
        {arriveTangent=-17.3697, arriveTangentWeight=0.0838, interpMode=RCIM_Cubic, leaveTangent=-17.3697, tangentMode=RCTM_User, time=1, value=0}
       ]}}}
   )
   &DynamicMaterialParameters(
    Index 0 Param 1=1
    Index 0 Param 3=1
    Index 0 Param 4=1
    Module.Index0Float0Write=true
    Module.Index0Float1Write=true
    Module.Index0Float2Write=true
    Module.Index0Float3Write=true
   )
   &ScaleColor(
    Scale Alpha=&FloatFromCurve(
     FloatCurve=DataInterface() {propertyValues {Curve {keys [
         {interpMode=RCIM_Cubic, tangentMode=RCTM_User, time=0, value=0}
         {interpMode=RCIM_Cubic, tangentMode=RCTM_User, time=0.01, value=1}
         {arriveTangentWeight=0.2755, interpMode=RCIM_Cubic, leaveTangentWeight=0.1907, tangentMode=RCTM_User, time=0.985, value=1}
         {arriveTangent=-2.5571, tangentMode=RCTM_User, time=1, value=0}
        ]}}}
     Scale Curve=&Multiply_Float(A=2, B=&FNTimeOfDayScalar(Night Scale=0.5, Morning Scale=0.8, Evening Scale=0.8))
    )
    Scale RGB=&MakeVectorFromLinearColorRGB(Color=<rgba 0.4399 0.0499 1 1>)
    ScaleA=true
    ScaleRGB=true
   )
   /SetVariables(
    Particles.GlowCameraOffset=5
    Particles.GlowColor=&ColorFromCurve(ColorCurve=DataInterface() {propertyValues {
       AlphaCurve {keys [<at 0.0017 0>,<at 0.0294 1 RCIM_Cubic>,<at 0.9522 1>,<at 1 0 RCIM_Cubic>]}
       BlueCurve {keys [
         <at 0.0081 4.9379>
         {arriveTangentWeight=0.1717, interpMode=RCIM_Cubic, leaveTangentWeight=1.4831, tangentMode=RCTM_User, time=0.0596, value=0.0411}
         <at 1 0.0318 RCIM_Cubic>
        ]}
       GreenCurve {keys [<at 0.0081 0>, {arriveTangent=0.0002, interpMode=RCIM_Cubic, leaveTangent=0.0002, time=0.0596, value=0.0002}, <at 1 0.0002 RCIM_Cubic>]}
       RedCurve {keys [
         <at 0.0081 5>
         {arriveTangentWeight=0.1753, interpMode=RCIM_Cubic, leaveTangentWeight=1.5141, tangentMode=RCTM_User, time=0.0596, value=0.0166}
         <at 1 0.0129 RCIM_Cubic>
        ]}
      }})
    Particles.GlowSize=&Multiply_Vector2D_ByFloat(
     Float=&FloatFromCurve(
      FloatCurve=DataInterface() {propertyValues {Curve {keys [
          {interpMode=RCIM_Cubic, leaveTangentWeight=0.5835, tangentMode=RCTM_User, time=0, value=0.4}
          {arriveTangentWeight=0.0191, interpMode=RCIM_Cubic, tangentMode=RCTM_User, time=0.0573, value=1.2881}
          {
           arriveTangent=-0.6122
           arriveTangentWeight=0.0142
           interpMode=RCIM_Cubic
           leaveTangent=-0.6122
           tangentMode=RCTM_User
           time=0.1
           value=1.0358
          }
          {
           arriveTangent=0.194
           arriveTangentWeight=0.1175
           interpMode=RCIM_Cubic
           leaveTangent=0.194
           leaveTangentWeight=0.1586
           tangentMode=RCTM_User
           time=0.4526
           value=1.0792
          }
          {
           arriveTangent=-0.6348
           arriveTangentWeight=0.1043
           interpMode=RCIM_Cubic
           leaveTangent=-0.6348
           leaveTangentWeight=0.0543
           tangentMode=RCTM_User
           time=0.7655
           value=1.0752
          }
          {arriveTangent=-4.5858, interpMode=RCIM_Cubic, leaveTangent=-4.5858, time=0.9285, value=0.6571}
          {arriveTangent=-17.3697, arriveTangentWeight=0.0838, interpMode=RCIM_Cubic, leaveTangent=-17.3697, tangentMode=RCTM_User, time=1, value=0}
         ]}}}
      Scale Curve=700
     )
     Vector2D=<1 1>
    )
   ) {isSetParameters=true}
   /SetVariables(Particles.DecalSize:Vector3f=&VectorFromCurve(
     VectorCurve=DataInterface() {propertyValues {
       XCurve {keys [<at 0 1 RCIM_Cubic>,<at 1 1 RCIM_Cubic>]}
       YCurve {keys [
         <at 0 0.4 RCIM_Cubic>
         {arriveTangentWeight=0.0191, interpMode=RCIM_Cubic, tangentMode=RCTM_User, time=0.0573, value=1.2881}
         {
          arriveTangent=-0.6122
          arriveTangentWeight=0.0142
          interpMode=RCIM_Cubic
          leaveTangent=-0.6122
          tangentMode=RCTM_User
          time=0.1
          value=1.0358
         }
         {
          arriveTangent=0.194
          arriveTangentWeight=0.1175
          interpMode=RCIM_Cubic
          leaveTangent=0.194
          leaveTangentWeight=0.1586
          tangentMode=RCTM_User
          time=0.4526
          value=1.0792
         }
         {
          arriveTangent=-0.304
          arriveTangentWeight=0.1043
          interpMode=RCIM_Cubic
          leaveTangent=-0.304
          leaveTangentWeight=0.0543
          tangentMode=RCTM_User
          time=0.6975
          value=1.0877
         }
         {arriveTangent=-1.4355, interpMode=RCIM_Cubic, leaveTangent=-1.4355, tangentMode=RCTM_User, time=0.8808, value=0.9662}
         {arriveTangent=-17.3697, arriveTangentWeight=0.0838, interpMode=RCIM_Cubic, leaveTangent=-17.3697, tangentMode=RCTM_User, time=1, value=0}
        ]}
       ZCurve {keys [
         <at 0 0.4 RCIM_Cubic>
         {arriveTangentWeight=0.0191, interpMode=RCIM_Cubic, tangentMode=RCTM_User, time=0.0573, value=1.2881}
         {
          arriveTangent=-0.6122
          arriveTangentWeight=0.0142
          interpMode=RCIM_Cubic
          leaveTangent=-0.6122
          tangentMode=RCTM_User
          time=0.1
          value=1.0358
         }
         {
          arriveTangent=0.194
          arriveTangentWeight=0.1175
          interpMode=RCIM_Cubic
          leaveTangent=0.194
          leaveTangentWeight=0.1586
          tangentMode=RCTM_User
          time=0.4526
          value=1.0792
         }
         {
          arriveTangent=-0.304
          arriveTangentWeight=0.1043
          interpMode=RCIM_Cubic
          leaveTangent=-0.304
          leaveTangentWeight=0.0543
          tangentMode=RCTM_User
          time=0.6975
          value=1.0877
         }
         {arriveTangent=-1.4355, interpMode=RCIM_Cubic, leaveTangent=-1.4355, tangentMode=RCTM_User, time=0.8808, value=0.9662}
         {arriveTangent=-17.3697, arriveTangentWeight=0.0838, interpMode=RCIM_Cubic, leaveTangent=-17.3697, tangentMode=RCTM_User, time=1, value=0}
        ]}
      }}
     Scale Curve=<500 1350 1350>
    )) {isSetParameters=true}
   /SetVariables(Particles.VortexScale:Vector3f=&Multiply_VectorByFloat(
     Vector=<1 1 1>
     Float=&FloatFromCurve(
      FloatCurve=DataInterface() {propertyValues {Curve {keys [
          {interpMode=RCIM_Cubic, leaveTangentWeight=0.5835, tangentMode=RCTM_User, time=0, value=0.4}
          {arriveTangentWeight=0.1242, interpMode=RCIM_Cubic, leaveTangentWeight=0.0922, tangentMode=RCTM_User, time=0.0573, value=1.2881}
          {
           arriveTangent=-0.6122
           arriveTangentWeight=0.0547
           interpMode=RCIM_Cubic
           leaveTangent=-0.6122
           leaveTangentWeight=0.2209
           tangentMode=RCTM_User
           time=0.1
           value=1.0358
          }
          {arriveTangent=0.194, interpMode=RCIM_Cubic, leaveTangent=0.194, tangentMode=RCTM_User, time=0.4526, value=1.0792}
          {
           arriveTangent=-0.304
           arriveTangentWeight=0.0816
           interpMode=RCIM_Cubic
           leaveTangent=-0.304
           leaveTangentWeight=0.0611
           tangentMode=RCTM_User
           time=0.6975
           value=1.0877
          }
          {
           arriveTangent=-1.4355
           arriveTangentWeight=0.1427
           interpMode=RCIM_Cubic
           leaveTangent=-1.4355
           leaveTangentWeight=0.0397
           tangentMode=RCTM_User
           time=0.8808
           value=0.9662
          }
          {arriveTangent=-17.3697, interpMode=RCIM_Cubic, leaveTangent=-17.3697, tangentMode=RCTM_User, time=1, value=0}
         ]}}}
      Scale Curve=5
     )
    )) {isSetParameters=true}
  ]
  rendererClasses [
   &NiagaraSpriteRendererProperties
   &NiagaraDecalRendererProperties
   &NiagaraMeshRendererProperties
  ]
  renderers [
   {rendererClass=&NiagaraSpriteRendererProperties}
   {rendererClass=&NiagaraDecalRendererProperties, rendererIndex=1}
   {rendererClass=&NiagaraMeshRendererProperties, rendererIndex=2}
  ]
 }
 PortalOutline {rendererClasses [], renderers []}
 Ring_Energy {
  emitterUpdateScript [
   &EmitterState(Life Cycle Mode=::NewEnumerator1, Loop Behavior=::NewEnumerator1, Loop Delay=0.15, Loop Duration=2.2, UseLoopDelay=true)
   &SpawnBurst_Instantaneous(Spawn Count=2, Use Spawn Probability=false, UseLoopCountLimit=false)
   &SpawnRate(
    SpawnRate=&FloatFromCurve(
     FloatCurve=DataInterface() {propertyValues {Curve {keys [
         {tangentMode=RCTM_User, time=0, value=1}
         {arriveTangent=-0.0405, interpMode=RCIM_Cubic, leaveTangent=-0.0405, tangentMode=RCTM_User, time=0.7009, value=1}
         {arriveTangent=-8.0089, interpMode=RCIM_Cubic, leaveTangent=-8.0089, tangentMode=RCTM_User, time=1, value=0}
        ]}}}
     CurveIndex=$Emitter.NormalizedLoopAge
     Scale Curve=20
    )
    Use Spawn Probability=false
   )
  ]
  particleSpawnScript [
   &InitializeParticle(
    Lifetime Max=1
    Lifetime Min=0.7
    Lifetime Mode=::NewEnumerator1
    Mesh Scale=&Multiply_VectorByFloat(
     Vector=&MakeVector(X=1, Y=1, Z=&RandomRangeFloat(Minimum=3, Maximum=4))
     Float=&RandomRangeFloat(Minimum=1.5, Maximum=2)
    )
    Mesh Scale Mode=::NewEnumerator4
    UsePositionOffset=true
    Write Material Random=true
   ) {module=InitializeParticle001}
   &SystemLocation(MaskSpawnGroup=false, Offset=&RandomRangeVector(Minimum=<0 0 0>, Maximum=<0 0 3>))
   &InitialMeshOrientation(
    Mesh Orientation Mode=::NewEnumerator2
    Orientation Coordinate Space=::Local
    Rotation=&RandomRangeVector(Minimum=<0 0 0>, Maximum=<0 0 1>)
    Use Rotation Vector=true
   ) {module=InitialMeshOrientation001}
   &AddVelocity(Velocity=&RandomRangeVector(Minimum=<0 0 0>, Maximum=<0 0 50>))
  ]
  particleUpdateScript [
   &SolveForcesAndVelocity(Clamp Velocity=false, Limit Acceleration=false)
   &ParticleState()
   &UpdateMeshOrientation(
    Rotation Rate=&SelectFloatFromArray(
     Array Sampling Mode=::NewEnumerator0
     Direct Array Index=?
     Float Selection Array=DataInterface() {propertyValues {FloatData [-0.3,0.3,0.2,-0.2]}}
     Random Seed=0
    )
    Rotation Vector=<0 0 1>
   )
   &DynamicMaterialParameters(
    Index 0 Param 1=&FloatFromCurve(
     FloatCurve=DataInterface() {propertyValues {Curve {keys [{interpMode=RCIM_Cubic, leaveTangent=-0.0324, tangentMode=RCTM_Break, time=0, value=1.5}, <at 1 -0.5>]}}}
     Scale Curve=&RandomRangeFloat(Minimum=0.8, Maximum=1)
    )
    Index 0 Param 2=&FloatFromCurve(FloatCurve=DataInterface() {propertyValues {Curve {keys [
         {arriveTangent=0.3262, interpMode=RCIM_Cubic, leaveTangent=0.3262, tangentMode=RCTM_User, time=0.2, value=0}
         {arriveTangent=1.6237, interpMode=RCIM_Cubic, leaveTangent=1.6237, tangentMode=RCTM_User, time=1, value=1}
        ]}}})
    Module.Index0Float0Write=true
    Module.Index0Float1Write=true
    Module.Index0Float2Write=true
    Module.Index0Float3Write=true
   )
   &ScaleMeshSize(Scale Factor=&Multiply_VectorByFloat(
     Vector=&Multiply_VectorByFloat(
      Vector=&MakeVector(
       X=1
       Y=1
       Z=&Multiply_Float(
        A=&FloatFromCurve(
         FloatCurve=DataInterface() {propertyValues {Curve {keys [{arriveTangent=-1.572, time=0, value=1}, {interpMode=RCIM_Cubic, tangentMode=RCTM_User, time=1, value=0.5}]}}}
         Scale Curve=0.5
        )
        B=&RandomRangeFloat(Minimum=1, Maximum=1)
       )
      )
      Float=0.8
     )
     Float=&FloatFromCurve(
      FloatCurve=DataInterface() {propertyValues {Curve {keys [
          {arriveTangent=4.7736, interpMode=RCIM_Cubic, leaveTangent=4.7736, tangentMode=RCTM_User, time=0, value=2.0364}
          {arriveTangent=0.9183, interpMode=RCIM_Cubic, leaveTangent=0.9183, tangentMode=RCTM_User, time=1, value=3.5}
         ]}}}
      Scale Curve=&Multiply_Float(
       A=1.7
       B=&FloatFromCurve(
        FloatCurve=DataInterface() {propertyValues {Curve {keys [{interpMode=RCIM_Cubic, tangentMode=RCTM_User, time=0, value=0.4}, {interpMode=RCIM_Cubic, tangentMode=RCTM_User, time=0.2095, value=1}]}}}
        CurveIndex=$Emitter.Age
       )
      )
     )
    ))
   &ScaleColor(
    Scale Alpha=&FloatFromCurve(
     FloatCurve=DataInterface() {propertyValues {Curve {keys [
         {arriveTangent=-0.0329, interpMode=RCIM_Cubic, leaveTangent=-0.0329, tangentMode=RCTM_User, time=0, value=0}
         {arriveTangent=3.3381, interpMode=RCIM_Cubic, leaveTangent=3.3381, tangentMode=RCTM_User, time=0.1, value=0.6}
         {arriveTangent=0.0085, interpMode=RCIM_Cubic, leaveTangent=0.0085, tangentMode=RCTM_User, time=0.2, value=0.75}
         {arriveTangent=-1.463, interpMode=RCIM_Cubic, leaveTangent=-1.463, tangentMode=RCTM_User, time=0.6, value=0.45}
         {arriveTangent=0.0414, interpMode=RCIM_Cubic, leaveTangent=0.0414, tangentMode=RCTM_User, time=1, value=0}
        ]}}}
     Scale Curve=&RandomRangeFloat(Minimum=1, Maximum=1.4)
    )
    Scale RGB=&MakeVectorFromLinearColorRGB(Color=&LerpLinearColors(
      StartColor=&SelectLinearColorFromArray(
       Array Sampling Mode=::NewEnumerator0
       Direct Array Index=?
       Color Selection Array=DataInterface() {propertyValues {ColorData [<rgba 0.0451 0.007 0.0457 1>, <rgba 1 0.1773 0.7967 1>]}}
       Random Seed=0
      )
      EndColor=<rgba 20 3.5306 10.2244 1>
      LerpFactor=&FloatFromCurve(
       FloatCurve=DataInterface() {propertyValues {Curve {keys [
           {arriveTangent=-7.9858, interpMode=RCIM_Cubic, leaveTangent=-7.9858, tangentMode=RCTM_User, time=0, value=1}
           {interpMode=RCIM_Cubic, tangentMode=RCTM_User, time=0.4067, value=0}
          ]}}}
       CurveIndex=$Emitter.Age
      )
     ))
    ScaleA=true
    ScaleRGB=true
   )
  ]
  rendererClasses [&NiagaraMeshRendererProperties]
  renderers [{rendererClass=&NiagaraMeshRendererProperties, rendererIndex=0}]
 }
 Ring_Line {
  emitterUpdateScript [
   &EmitterState(Life Cycle Mode=::NewEnumerator1, Loop Behavior=::NewEnumerator1, Loop Delay=0.4, Loop Duration=0.8, UseLoopDelay=true)
   &SpawnRate(SpawnRate=12, Use Spawn Probability=false)
  ]
  particleSpawnScript [
   &InitializeParticle(
    Color=<rgba 1 1 1 0.6734>
    Lifetime Max=0.8
    Lifetime Min=0.5
    Lifetime Mode=::NewEnumerator1
    Mesh Scale=&Multiply_VectorByFloat(
     Vector=&MakeVector(X=1, Y=1, Z=&RandomRangeFloat(Minimum=10, Maximum=12))
     Float=&RandomRangeFloat(Minimum=3, Maximum=3.5)
    )
    Mesh Scale Mode=::NewEnumerator4
    Position Offset=<0 0 30>
    UsePositionOffset=true
    Write Material Random=true
   ) {module=InitializeParticle001}
   &SystemLocation(MaskSpawnGroup=false, Offset=&RandomRangeVector(Minimum=<0 0 0>, Maximum=<0 0 10>))
   &InitialMeshOrientation(
    Mesh Orientation Mode=::NewEnumerator2
    Orientation Coordinate Space=::Local
    Rotation=&RandomRangeVector(Minimum=<0 -0.05 0>, Maximum=<0 0.05 1>)
    Use Rotation Vector=true
   ) {module=InitialMeshOrientation001}
  ]
  particleUpdateScript [
   &ParticleState()
   &UpdateMeshOrientation(
    Rotation Rate=&SelectFloatFromArray(
     Array Sampling Mode=::NewEnumerator0
     Direct Array Index=?
     Float Selection Array=DataInterface() {propertyValues {FloatData [-0.3,0.3,0.2,-0.2]}}
     Random Seed=0
    )
    Rotation Vector=<0 0 1>
   )
   &ScaleMeshSize(Scale Factor=&VectorFromFloat(Value=3.5)) {module=ScaleMeshSize001}
   &ScaleMeshSize(Scale Factor=&Multiply_VectorByFloat(
     Vector=&Multiply_VectorByFloat(
      Vector=&MakeVector(
       X=1
       Y=1
       Z=&Multiply_Float(
        A=&FloatFromCurve(
         FloatCurve=DataInterface() {propertyValues {Curve {keys [{arriveTangent=-1.572, time=0, value=1}, {interpMode=RCIM_Cubic, tangentMode=RCTM_User, time=1, value=0.5}]}}}
         Scale Curve=0.7
        )
        B=&RandomRangeFloat(Minimum=1, Maximum=1)
       )
      )
      Float=0.8
     )
     Float=&FloatFromCurve(
      FloatCurve=DataInterface() {propertyValues {Curve {keys [
          {arriveTangent=2.4838, interpMode=RCIM_Cubic, leaveTangent=2.4838, tangentMode=RCTM_User, time=0, value=1.295}
          {arriveTangent=0.3815, interpMode=RCIM_Cubic, leaveTangent=0.3815, tangentMode=RCTM_User, time=1, value=2}
         ]}}}
      Scale Curve=&FN_FrontEndFloatModifier(Gameplay=1.7, Front End=1.5)
     )
    ))
   &ScaleColor(
    Scale Alpha=&FloatFromCurve(
     FloatCurve=DataInterface() {propertyValues {Curve {keys [
         {arriveTangent=6.8924, interpMode=RCIM_Cubic, leaveTangent=6.8924, tangentMode=RCTM_User, time=0, value=0}
         {arriveTangent=3.3638, interpMode=RCIM_Cubic, leaveTangent=3.3638, tangentMode=RCTM_User, time=0.1, value=0.6}
         {arriveTangent=0.0044, interpMode=RCIM_Cubic, leaveTangent=0.0044, tangentMode=RCTM_User, time=0.2, value=0.75}
         {arriveTangent=-1.0842, interpMode=RCIM_Cubic, leaveTangent=-1.0842, tangentMode=RCTM_User, time=0.6953, value=0.5076}
         {arriveTangent=-1.8777, interpMode=RCIM_Cubic, leaveTangent=-1.8777, tangentMode=RCTM_User, time=1, value=0}
        ]}}}
     Scale Curve=2
    )
    Scale RGB=&MakeVectorFromLinearColorRGB(Color=&ColorFromCurve(ColorCurve=DataInterface() {propertyValues {
        AlphaCurve {keys []}
        BlueCurve {keys [
          {arriveTangent=0.3322, interpMode=RCIM_Cubic, leaveTangent=0.3322, tangentMode=RCTM_User, time=0, value=0.555}
          <at 0.4237 0.328>
          <at 0.7343 0.1615>
         ]}
        GreenCurve {keys [
          {arriveTangent=-1.1199, interpMode=RCIM_Cubic, leaveTangent=-1.1199, tangentMode=RCTM_User, time=0, value=0.1167}
          <at 0.4237 0.007>
          <at 0.7343 0.0113>
         ]}
        RedCurve {keys [
          {arriveTangent=-3.8908, interpMode=RCIM_Cubic, leaveTangent=-3.8908, tangentMode=RCTM_User, time=0, value=0.3979}
          <at 0.4237 0.1318>
          <at 0.7343 0.0168>
         ]}
       }}))
    ScaleA=true
    ScaleRGB=true
   )
  ]
  rendererClasses [&NiagaraMeshRendererProperties]
  renderers [{rendererClass=&NiagaraMeshRendererProperties, rendererIndex=0}]
 }
 SmokeRing {
  emitterUpdateScript [
   &EmitterState(Life Cycle Mode=::NewEnumerator1, Loop Behavior=::NewEnumerator1, Loop Delay=0.15, Loop Duration=2.1, UseLoopDelay=true)
   &SpawnBurst_Instantaneous(Spawn Count=4, Use Spawn Probability=false, UseLoopCountLimit=false)
   &SpawnRate(
    SpawnRate=&FloatFromCurve(
     FloatCurve=DataInterface() {propertyValues {Curve {keys [
         {arriveTangent=0.2741, interpMode=RCIM_Cubic, leaveTangent=0.2741, tangentMode=RCTM_User, time=0, value=1}
         {arriveTangent=0.0207, interpMode=RCIM_Cubic, leaveTangent=0.0207, tangentMode=RCTM_User, time=0.822, value=0.9854}
        ]}}}
     CurveIndex=$Emitter.NormalizedLoopAge
     Scale Curve=7
    )
    Use Spawn Probability=false
   )
  ]
  particleSpawnScript [
   &InitializeParticle(
    Lifetime Max=1
    Lifetime Min=0.8
    Lifetime Mode=::NewEnumerator1
    Mesh Scale=&Multiply_VectorByFloat(
     Vector=&RandomRangeVector(Minimum=<1 1 1>, Maximum=<1 1 1.5>)
     Float=&RandomRangeFloat(Minimum=2, Maximum=3)
    )
    Mesh Scale Mode=::NewEnumerator4
    UsePositionOffset=false
    Write Material Random=true
   )
   &InitialMeshOrientation(
    Mesh Orientation Mode=::NewEnumerator2
    Rotation=&RandomRangeVector(Minimum=<0.03 0.03 -1>, Maximum=<0.03 0.03 1>)
    Use Rotation Vector=true
   )
   &ShapeLocation(
    Cylinder Height=3
    Cylinder Radius=1
    MaskSpawnGroup=false
    Offset=<0 0 5>
    Offset Mode=::NewEnumerator0
    Override Seed=false
    Shape Primitive=::NewEnumerator1
    Surface Only=false
    UseOwnerScale=true
   )
   &AddVelocity(Velocity=&RandomRangeVector(Minimum=<0 0 0>, Maximum=<0 0 50>))
  ]
  particleUpdateScript [
   &ParticleState()
   &DynamicMaterialParameters(
    Index 0 Param 1=&FloatFromCurve(FloatCurve=DataInterface() {propertyValues {Curve {keys [{interpMode=RCIM_Cubic, tangentMode=RCTM_User, time=0, value=0.8}, {interpMode=RCIM_Cubic, tangentMode=RCTM_User, time=1, value=0}]}}})
    Module.Index0Float0Write=true
    Module.Index0Float1Write=true
    Module.Index0Float2Write=true
    Module.Index0Float3Write=true
   )
   &ScaleMeshSize(Scale Factor=&Multiply_VectorByFloat(
     Vector=&Multiply_VectorByFloat(Vector=$Engine.Owner.Scale, Float=&FN_FrontEndFloatModifier(Gameplay=0.9, Front End=0.8))
     Float=&FloatFromCurve(
      FloatCurve=DataInterface() {propertyValues {Curve {keys [
          {arriveTangent=0.0893, interpMode=RCIM_Cubic, leaveTangent=0.0893, tangentMode=RCTM_User, time=0, value=0.55}
          {arriveTangent=0.0158, interpMode=RCIM_Cubic, leaveTangent=0.0158, tangentMode=RCTM_User, time=1, value=1}
         ]}}}
      Scale Curve=&Multiply_Float(
       A=4
       B=&FloatFromCurve(
        FloatCurve=DataInterface() {propertyValues {Curve {keys [{interpMode=RCIM_Cubic, tangentMode=RCTM_User, time=0, value=0.4}, {interpMode=RCIM_Cubic, tangentMode=RCTM_User, time=0.2095, value=1}]}}}
        CurveIndex=$Emitter.Age
       )
      )
     )
    ))
   &UpdateMeshOrientation(
    Rotation Rate=&SelectFloatFromArray(
     Array Sampling Mode=::NewEnumerator0
     Direct Array Index=?
     Float Selection Array=DataInterface() {propertyValues {FloatData [-0.3,0.3,0.2,-0.2]}}
     Random Seed=0
    )
    Rotation Vector=<0 0 1>
   )
   &ScaleColor(
    Scale Alpha=&FloatFromCurve(
     FloatCurve=DataInterface() {propertyValues {Curve {keys [
         {arriveTangent=1.0939, interpMode=RCIM_Cubic, leaveTangent=1.0939, tangentMode=RCTM_User, time=0, value=0}
         {arriveTangent=-0.0731, interpMode=RCIM_Cubic, leaveTangent=-0.0731, tangentMode=RCTM_User, time=0.326, value=1}
         {arriveTangent=-0.05, interpMode=RCIM_Cubic, leaveTangent=-0.8744, tangentMode=RCTM_Break, time=0.4, value=1}
         {arriveTangent=-0.5359, tangentMode=RCTM_User, time=1, value=0}
        ]}}}
     Scale Curve=0.8
    )
    Scale RGB=&MakeVectorFromLinearColorRGB(Color=&LerpLinearColors(
      StartColor=&ColorFromCurve(ColorCurve=DataInterface() {propertyValues {
         AlphaCurve {keys []}
         BlueCurve {keys [
           {arriveTangent=0.3322, interpMode=RCIM_Cubic, leaveTangent=0.3322, tangentMode=RCTM_User, time=0.1294, value=0.9578}
           <at 0.4879 0.0133>
           <at 0.7343 0.0503>
          ]}
         GreenCurve {keys [
           {arriveTangent=-1.1199, interpMode=RCIM_Cubic, leaveTangent=-1.1199, tangentMode=RCTM_User, time=0.1294, value=0.1424}
           <at 0.4879 0.0039>
           <at 0.7343 0.0258>
          ]}
         RedCurve {keys [
           {arriveTangent=-3.8908, interpMode=RCIM_Cubic, leaveTangent=-3.8908, tangentMode=RCTM_User, time=0.1294, value=0.5264}
           <at 0.4879 0.0076>
           <at 0.7343 0.0165>
          ]}
        }})
      EndColor=<rgba 11.5381 1.6095 20 1>
      LerpFactor=&FloatFromCurve(
       FloatCurve=DataInterface() {propertyValues {Curve {keys [
           {arriveTangent=-5.4468, interpMode=RCIM_Cubic, leaveTangent=-5.4468, tangentMode=RCTM_User, time=0, value=1}
           {interpMode=RCIM_Cubic, tangentMode=RCTM_User, time=0.4046, value=0.01}
          ]}}}
       CurveIndex=$Emitter.Age
      )
     ))
    ScaleA=true
    ScaleRGB=true
   )
   &SolveForcesAndVelocity(Clamp Velocity=false, Limit Acceleration=false)
  ]
  rendererClasses [&NiagaraMeshRendererProperties]
  renderers [{rendererClass=&NiagaraMeshRendererProperties, rendererIndex=0}]
 }
 Trail_ring {
  emitterUpdateScript [
   &EmitterState(Life Cycle Mode=::NewEnumerator1, Loop Behavior=::NewEnumerator1, Loop Delay=0.15, Loop Duration=1, UseLoopDelay=false)
   &SpawnBurst_Instantaneous(Spawn Count=2, Use Spawn Probability=false, UseLoopCountLimit=false)
  ]
  particleSpawnScript [
   &PartitionParticles()
   &InitializeParticle(
    Lifetime=0.5
    Mesh Scale=&Multiply_VectorByFloat(Vector=<1 0.6 1>, Float=5)
    Mesh Scale Mode=::NewEnumerator4
    Position Offset=<0 0 20>
    UsePositionOffset=true
    Write Material Random=true
   )
   &InitialMeshOrientation(
    Mesh Axis To Orient=&MaskVector(Mask Parameter=$Particles.PartitionParticles.Partition, Mask=0, Masked Vector=<0 1 0>, Passthrough Vector=<0 -1 0>)
    Mesh Orientation Mode=::NewEnumerator2
    Orientation Vector=<0 0 1>
    Rotation=&MaskVector(Mask Parameter=$Particles.PartitionParticles.Partition, Mask=0, Masked Vector=<0 0.438 0>, Passthrough Vector=<0 0.9897 0>)
    Use Rotation Vector=true
   )
   &DynamicMaterialParameters(
    Index 0 Param 1=&Multiply_Float(
     A=&RandomRangeFloat(Minimum=30, Maximum=30)
     B=&FloatFromCurve(
      FloatCurve=DataInterface() {propertyValues {Curve {keys [
          <at 1 2.5 RCIM_Cubic>
          {arriveTangent=-1.44, interpMode=RCIM_Cubic, leaveTangent=-1.44, time=1.6444, value=1.8}
          {arriveTangent=-0.0085, tangentMode=RCTM_User, time=2.25, value=0.7}
         ]}}}
      CurveIndex=$Emitter.Age
      Scale Curve=0
     )
    )
    Index 0 Param 2=&Multiply_Float(
     A=&RandomRangeFloat(Minimum=20, Maximum=30)
     B=&FloatFromCurve(
      FloatCurve=DataInterface() {propertyValues {Curve {keys [
          <at 1 2.5 RCIM_Cubic>
          {arriveTangent=-1.44, interpMode=RCIM_Cubic, leaveTangent=-1.44, time=1.6444, value=1.8}
          {arriveTangent=-0.0085, tangentMode=RCTM_User, time=2.25, value=0.7}
         ]}}}
      CurveIndex=$Emitter.Age
      Scale Curve=0
     )
    )
    Index 0 Param 3=1.5
    Index 0 Param 4=&RandomRangeFloat(Minimum=0, Maximum=1)
    Index 1 Param 1=&RandomRangeFloat(Minimum=0.2, Maximum=1)
    Index 1 Param 2=&FloatFromCurve(FloatCurve=DataInterface() {propertyValues {Curve {keys [{tangentMode=RCTM_User, time=0, value=1}, {tangentMode=RCTM_User, time=0.5, value=1}]}}})
    Module.Index0Float0Write=true
    Module.Index0Float1Write=true
    Module.Index0Float2Write=true
    Module.Index0Float3Write=true
    Module.Index1Float0Write=true
    Module.Index1Float1Write=true
    Module.Index1Float2Write=false
    Module.Index1Float3Write=false
    Write Parameter Index 1=::NewEnumerator1
   )
   &CameraOffset(Camera Offset Amount=5)
  ]
  particleUpdateScript [
   &ParticleState()
   &UpdateMeshOrientation(
    Rotation Rate=&MaskFloat(
     Mask Parameter=$Particles.PartitionParticles.Partition
     Mask=0
     Masked Float=&FloatFromCurve(FloatCurve=DataInterface() {propertyValues {Curve {keys [<at 0 1>,<at 1 0>]}}})
     Passthrough Float=&FloatFromCurve(FloatCurve=DataInterface() {propertyValues {Curve {keys [{tangentMode=RCTM_User, time=0, value=0}, {tangentMode=RCTM_User, time=1, value=1}]}}})
    )
    Rotation Vector=<0 1 0>
   )
   &ScaleMeshSize(Scale Factor=&Multiply_VectorByFloat(
     Vector=<1 1 1>
     Float=&Multiply_Float(
      A=0.8
      B=&FloatFromCurve(FloatCurve=DataInterface() {propertyValues {Curve {keys [{interpMode=RCIM_Cubic, tangentMode=RCTM_User, time=0, value=0}, {interpMode=RCIM_Cubic, tangentMode=RCTM_User, time=0.4414, value=1.3164}]}}})
     )
    ))
   &ScaleColor(
    Scale Alpha=&FloatFromCurve(
     FloatCurve=DataInterface() {propertyValues {Curve {keys [
         {arriveTangent=4.6142, interpMode=RCIM_Cubic, leaveTangent=4.6142, tangentMode=RCTM_User, time=0, value=0}
         {interpMode=RCIM_Cubic, tangentMode=RCTM_User, time=0.5, value=1}
         {arriveTangent=-2, interpMode=RCIM_Cubic, leaveTangent=-2, time=0.8764, value=0.8105}
         {arriveTangent=-5.0551, interpMode=RCIM_Cubic, leaveTangent=-5.0551, tangentMode=RCTM_User, time=1, value=0}
        ]}}}
     Scale Curve=0.7013
    )
    Scale RGB=&Multiply_VectorByFloat(
     Vector=&MakeVectorFromLinearColorRGB(Color=&ColorFromCurve(ColorCurve=DataInterface() {propertyValues {
         AlphaCurve {keys [<at 0 1 RCIM_Cubic>,<at 1 0 RCIM_Cubic>]}
         BlueCurve {keys [<at 0 1 RCIM_Cubic>,<at 0.4599 1>,<at 1 1 RCIM_Cubic>]}
         GreenCurve {keys [<at 0 0.3293 RCIM_Cubic>,<at 0.4599 0.0342>,<at 1 0 RCIM_Cubic>]}
         RedCurve {keys [<at 0 0.721 RCIM_Cubic>,<at 0.4599 0.5859>,<at 1 0.231 RCIM_Cubic>]}
        }}))
     Float=1
    )
    ScaleA=true
    ScaleRGB=true
   )
  ]
  rendererClasses [&NiagaraMeshRendererProperties]
  renderers [{rendererClass=&NiagaraMeshRendererProperties, rendererIndex=0}]
 }
 Trail_ring001 {
  emitterUpdateScript [
   &EmitterState(Life Cycle Mode=::NewEnumerator1, Loop Behavior=::NewEnumerator1, Loop Delay=0.15, Loop Duration=1.3, UseLoopDelay=true)
   &SpawnBurst_Instantaneous(Use Spawn Probability=false, UseLoopCountLimit=false)
   &SpawnRate(SpawnRate=5, Use Spawn Probability=false)
  ]
  particleSpawnScript [
   &InitializeParticle(
    Color=&FNTimeOfDayColor(
     Night Color=<rgba 0.0571 0.0245 0.217 1>
     Morning Color=<rgba 0.085 0.0228 0.217 1>
     Day Color=<rgba 0.085 0.0228 0.217 1>
     Evening Color=<rgba 0.085 0.0228 0.217 1>
    )
    Lifetime=1.3
    Mesh Scale=&Multiply_VectorByFloat(Vector=<1 1 1>, Float=&RandomRangeFloat(Minimum=5, Maximum=7.5))
    Mesh Scale Mode=::NewEnumerator4
    Position Offset=&RandomRangeVector(Minimum=<0 0 5>, Maximum=<0 0 50>)
    UsePositionOffset=true
    Write Material Random=true
   )
   &InitialMeshOrientation(
    Mesh Axis To Orient=<0 1 0>
    Mesh Orientation Mode=::NewEnumerator2
    Orientation Vector=<0 0 1>
    Rotation=&RandomRangeVector(Minimum=<0 0 0>, Maximum=<0 1 0>)
    Use Rotation Vector=true
   )
   &DynamicMaterialParameters(
    Index 0 Param 1=&Multiply_Float(
     A=&RandomRangeFloat(Minimum=30, Maximum=30)
     B=&FloatFromCurve(
      FloatCurve=DataInterface() {propertyValues {Curve {keys [
          <at 1 2.5 RCIM_Cubic>
          {arriveTangent=-1.44, interpMode=RCIM_Cubic, leaveTangent=-1.44, time=1.6444, value=1.8}
          {arriveTangent=-0.0085, tangentMode=RCTM_User, time=2.25, value=0.7}
         ]}}}
      CurveIndex=$Emitter.Age
      Scale Curve=0
     )
    )
    Index 0 Param 2=&Multiply_Float(
     A=&RandomRangeFloat(Minimum=20, Maximum=30)
     B=&FloatFromCurve(
      FloatCurve=DataInterface() {propertyValues {Curve {keys [
          <at 1 2.5 RCIM_Cubic>
          {arriveTangent=-1.44, interpMode=RCIM_Cubic, leaveTangent=-1.44, time=1.6444, value=1.8}
          {arriveTangent=-0.0085, tangentMode=RCTM_User, time=2.25, value=0.7}
         ]}}}
      CurveIndex=$Emitter.Age
      Scale Curve=0
     )
    )
    Index 0 Param 3=1.5
    Index 0 Param 4=&RandomRangeFloat(Minimum=0, Maximum=1)
    Index 1 Param 1=&RandomRangeFloat(Minimum=0.2, Maximum=1)
    Index 1 Param 2=&FloatFromCurve(FloatCurve=DataInterface() {propertyValues {Curve {keys [{tangentMode=RCTM_User, time=0, value=1}, {tangentMode=RCTM_User, time=0.5, value=1}]}}})
    Module.Index0Float0Write=true
    Module.Index0Float1Write=true
    Module.Index0Float2Write=true
    Module.Index0Float3Write=true
    Module.Index1Float0Write=true
    Module.Index1Float1Write=true
    Module.Index1Float2Write=false
    Module.Index1Float3Write=false
    Write Parameter Index 1=::NewEnumerator1
   )
   &CameraOffset(Camera Offset Amount=5)
  ]
  particleUpdateScript [
   &ParticleState()
   &ScaleMeshSize(Scale Factor=&Multiply_VectorByFloat(
     Vector=<1 1 1>
     Float=&Multiply_Float(
      A=0.8
      B=&FloatFromCurve(
       FloatCurve=DataInterface() {propertyValues {Curve {keys [<at 0 0.8>,<at 1 1.2>]}}}
       Scale Curve=&FloatFromCurve(
        FloatCurve=DataInterface() {propertyValues {Curve {keys [{interpMode=RCIM_Cubic, tangentMode=RCTM_User, time=0, value=0.4}, {interpMode=RCIM_Cubic, tangentMode=RCTM_User, time=0.2095, value=1}]}}}
        CurveIndex=$Emitter.Age
       )
      )
     )
    ))
   &ScaleColor(
    Scale Alpha=&FloatFromCurve(
     FloatCurve=DataInterface() {propertyValues {Curve {keys [
         {arriveTangent=4.6142, interpMode=RCIM_Cubic, leaveTangent=4.6142, tangentMode=RCTM_User, time=0, value=0}
         {interpMode=RCIM_Cubic, tangentMode=RCTM_User, time=0.5, value=1}
         {arriveTangent=-2, interpMode=RCIM_Cubic, leaveTangent=-2, time=0.8764, value=0.8105}
         {arriveTangent=-5.0551, interpMode=RCIM_Cubic, leaveTangent=-5.0551, tangentMode=RCTM_User, time=1, value=0}
        ]}}}
     Scale Curve=&FloatFromCurve(
      FloatCurve=DataInterface() {propertyValues {Curve {keys [<at 0.7 1>,<at 2.5753 1>,<at 2.7698 0>]}}}
      CurveIndex=$Emitter.Age
      Scale Curve=0.8
     )
    )
    Scale RGB=&Multiply_VectorByFloat(Vector=<1 1 1>, Float=1)
    ScaleA=true
    ScaleRGB=true
   )
  ]
  rendererClasses [&NiagaraMeshRendererProperties]
  renderers [{rendererClass=&NiagaraMeshRendererProperties, rendererIndex=0}]
 }
 WispySmoke {
  emitterUpdateScript [
   &EmitterState(Life Cycle Mode=::NewEnumerator1, Loop Behavior=::NewEnumerator1, Loop Delay=0.8, Loop Duration=2, UseLoopDelay=true)
   &SpawnRate(SpawnRate=15, Use Spawn Probability=false)
   &SpawnBurst_Instantaneous(Spawn Count=4, Use Spawn Probability=false, UseLoopCountLimit=false)
  ]
  particleSpawnScript [
   &InitializeParticle(
    Color Maximum=<rgba 0.7216 0.1692 0.4756 1>
    Color Minimum=<rgba 0.1568 0.0307 0.1115 1>
    Color Mode=::NewEnumerator2
    Lifetime Max=0.6
    Lifetime Min=0.4
    Lifetime Mode=::NewEnumerator1
    Sprite Rotation Mode=::NewEnumerator2
    Sprite Size Mode=::NewEnumerator1
    Uniform Sprite Size Max=500
    Uniform Sprite Size Min=350
    UsePositionOffset=false
    Write Material Random=true
   )
   &ShapeLocation(Handle Radius=0, Large Radius=250, MaskSpawnGroup=false, Override Seed=false, Shape Primitive=::NewEnumerator3, UseOwnerScale=true)
   &CameraOffset(Camera Offset Amount=30)
   &AddVelocity(Velocity=&RandomRangeVector(Minimum=<0 0 100>, Maximum=<0 125 50>))
  ]
  particleUpdateScript [
   &ParticleState()
   &SpriteRotationRate(Rotation Rate=&RandomRangeFloat(Minimum=-50, Maximum=50))
   &ScaleSpriteSize(
    Scale Sprite Size Mode=::NewEnumerator0
    Uniform Scale Factor=&FloatFromCurve(FloatCurve=DataInterface() {propertyValues {Curve {keys [<at 0 0.8>,<at 1 1.2>]}}})
   )
   &SubUVAnimation(
    End Frame Range Override=30
    "Sprite Renderer "=DataInterface() {propertyValues {SpriteRenderer=&self:WispySmoke_0.NiagaraSpriteRendererProperties_0}}
    Start Frame Range Override=&RandomRangeInt(Minimum=0, Maximum=5)
    UseEndFrame=true
    UseStartFrame=true
   )
   &ScaleColor(
    Scale Alpha=&Multiply_Float(
     A=&RandomRangeFloat(Minimum=0.5, Maximum=1)
     B=&FloatFromCurve(
      FloatCurve=DataInterface() {propertyValues {Curve {keys [
          {arriveTangent=6.7463, interpMode=RCIM_Cubic, leaveTangent=6.7463, tangentMode=RCTM_User, time=0, value=0}
          {arriveTangent=2, interpMode=RCIM_Cubic, leaveTangent=2, time=0.0827, value=0.7619}
          {interpMode=RCIM_Cubic, tangentMode=RCTM_User, time=0.5, value=1}
          {arriveTangent=-5.5093, interpMode=RCIM_Cubic, leaveTangent=-5.5093, tangentMode=RCTM_User, time=1, value=0}
         ]}}}
      Scale Curve=&FloatFromCurve(
       FloatCurve=DataInterface() {propertyValues {Curve {keys [
           {tangentMode=RCTM_User, time=0, value=1}
           {arriveTangent=-0.0405, interpMode=RCIM_Cubic, leaveTangent=-0.0405, tangentMode=RCTM_User, time=1.5916, value=1}
           {arriveTangent=-2.5083, interpMode=RCIM_Cubic, leaveTangent=-2.5083, tangentMode=RCTM_User, time=2.4445, value=0}
          ]}}}
       CurveIndex=$Emitter.Age
      )
     )
    )
    ScaleA=true
    ScaleRGB=true
   )
   &SolveForcesAndVelocity(Clamp Velocity=false, Limit Acceleration=false)
  ]
  rendererClasses [&NiagaraSpriteRendererProperties]
  renderers [{rendererClass=&NiagaraSpriteRendererProperties}]
 }
}
summary {
 systemName=NS_TentaclePortal
 emitters [
  {emitterName=WispySmoke, rendererClasses [&NiagaraSpriteRendererProperties]}
  {emitterName=Ring_Energy, rendererClasses [&NiagaraMeshRendererProperties]}
  {emitterName=SmokeRing, rendererClasses [&NiagaraMeshRendererProperties]}
  {emitterName=Ring_Line, rendererClasses [&NiagaraMeshRendererProperties]}
  {emitterName=Trail_ring001, rendererClasses [&NiagaraMeshRendererProperties]}
  {emitterName=Glow, rendererClasses [&NiagaraSpriteRendererProperties, &NiagaraLightRendererProperties]}
  {
   emitterName=Portal
   rendererClasses [
    &NiagaraSpriteRendererProperties
    &NiagaraDecalRendererProperties
    &NiagaraMeshRendererProperties
   ]
  }
  {emitterName=OpenPortal, rendererClasses [&NiagaraMeshRendererProperties]}
  {emitterName=Trail_ring, rendererClasses [&NiagaraMeshRendererProperties]}
  {emitterName=PortalOutline, rendererClasses []}
 ]
 userVariables [
  User.bIsFrontEnd {type=NiagaraBool}
  User.bIsFrontEndPreview {type=NiagaraBool}
  User.ProjStopped {type=NiagaraBool}
 ]
}
