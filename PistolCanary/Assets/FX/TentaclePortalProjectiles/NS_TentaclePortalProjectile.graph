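# Niagara system for the tentacle portal projectile: a purple void orb with chromatic refraction, light glow, smoke wisps and a layered ribbon trail.
# The orb and refraction smooth-lerp toward the owner's position, and the orb mesh orients and stretches along owner velocity.
# System timeline values taper smoke and trail and scale the orb's fade, alongside a User.ProjStopped flag.
# Three velocity debug emitters are disabled.
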NiagaraSystem /PistolCanary/Assets/FX/TentaclePortalProjectiles/NS_TentaclePortalProjectile
dependencies {
 usedDataInterfaces [
  &NiagaraDataInterfaceCurve
  &NiagaraDataInterfaceColorCurve
  &NiagaraDataInterfaceVectorCurve
  &NiagaraDataInterfaceArrayColor
  &NiagaraDataInterfaceArrayInt32
  &NiagaraDataInterfaceArrayFloat
  &NiagaraDataInterfaceSpriteRendererInfo
 ]
 usedDynamicInputs []
 usedModules [
  &self:SystemSpawnScript.SystemScriptSource.SystemScriptGraph.NiagaraNodeAssignment_1.SetVariables_#1
  &SystemState
  &self:SystemSpawnScript.SystemScriptSource.SystemScriptGraph.NiagaraNodeAssignment_0.SetVariables_#2
 ]
}
emitters {
 ChromaticRefraction {
  emitterUpdateScript [
   &EmitterState(
    Inactive Response=::NewEnumerator1
    Life Cycle Mode=::NewEnumerator1
    Loop Behavior=::NewEnumerator1
    Loop Delay=0.5
    Loop Duration Mode=::NewEnumerator1
    UseLoopDelay=true
   )
   &SpawnBurst_Instantaneous(Use Spawn Probability=false, UseLoopCountLimit=false)
  ]
  particleSpawnScript [&InitializeParticle(Sprite Size=<20 20>, Sprite Size Mode=::NewEnumerator4, UsePositionOffset=false, Write Material Random=true)]
  particleUpdateScript [
   &ParticleState(Kill Particles When Lifetime Has Elapsed=false, Let Infinitely Lived Particles Die When Emitter Deactivates=true)
   &Timeline(
    Play Rate=&FloatFromCurve(
     FloatCurve=DataInterface() {propertyValues {Curve {keys [
         {interpMode=RCIM_Cubic, leaveTangent=-12.5334, tangentMode=RCTM_Break, time=0, value=4}
         {interpMode=RCIM_Cubic, tangentMode=RCTM_Break, time=0.7, value=1}
        ]}}}
     Scale Curve=0.5
    )
    Timeline End=1.0000000200408773e+20
   )
   &ScaleSpriteSize(
    Scale Sprite Size Mode=::NewEnumerator0
    Uniform Scale Factor=&FloatFromCurve(
     FloatCurve=DataInterface() {propertyValues {Curve {keys [{interpMode=RCIM_Cubic, tangentMode=RCTM_User, time=0.5, value=0}, {interpMode=RCIM_Cubic, tangentMode=RCTM_User, time=1, value=1}]}}}
     CurveIndex=$Emitter.Age
     Scale Curve=70
    )
   )
   &DynamicMaterialParameters(
    Index 0 Param 1=2
    Index 1 Param 2=500
    Index 1 Param 4=1
    Index 2 Param 1=$Particles.Timeline.Time
    Index 2 Param 2=$Particles.Timeline.Time
    Index 2 Param 3=$Particles.Timeline.Time
    Index 2 Param 4=&SelectFloatFromArray(Direct Array Index=$Particles.PartitionParticles.Partition, Float Selection Array=DataInterface() {propertyValues {FloatData [1,0.3,0.3]}})
    Index 3 Param 1=&FloatFromCurve(FloatCurve=DataInterface() {propertyValues {Curve {keys [<at 0.1 1>,<at 0.4 0>]}}})
    Index 3 Param 2=200
    Module.Index0Param0Write=true
    Module.Index0Param1Write=true
    Module.Index0Param2Write=true
    Module.Index0Param3Write=true
    Module.Index1Param0Write=true
    Module.Index1Param1Write=true
    Module.Index1Param2Write=true
    Module.Index1Param3Write=true
    Module.Index2Param0Write=true
    Module.Index2Param1Write=true
    Module.Index2Param2Write=true
    Module.Index2Param3Write=true
    Module.Index3Param0Write=true
    Module.Index3Param1Write=true
    Module.Index3Param2Write=true
    Module.Index3Param3Write=true
    Write Parameter Index 1=true
    Write Parameter Index 2=true
    Write Parameter Index 3=true
   ) {module=DynamicMaterialParameters001}
   &ScaleColor(Scale RGB=&VectorFromFloat(Value=0.5), ScaleA=true, ScaleRGB=true) {module=ScaleColor001}
   /SetVariables(Particles.Position=&AddVectorToPosition(Position=$Engine.Emitter.SimulationPosition, Vector=&Multiply_Vector(A=$Engine.Owner.Velocity, B=<-0.08 1 1>))) {enabled=false, isSetParameters=true}
   /SetVariables(Particles.Position=/Niagara/DynamicInputs/Lerp/SmoothLerpOverTime_Position(
     Convergence Rate:NiagaraFloat=4.5
     Smooth Value:NiagaraPosition=$Particles.Initial.Position
     Target Value:NiagaraPosition=$Engine.Owner.Position
    )) {isSetParameters=true}
  ]
  rendererClasses [&NiagaraSpriteRendererProperties]
  renderers [{rendererClass=&NiagaraSpriteRendererProperties}]
 }
 FN_Debug {
  bEnabled=false
  emitterUpdateScript [
   &EmitterState()
   &SpawnBurst_Instantaneous(Spawn Count=1, Use Spawn Probability=false)
  ]
  particleSpawnScript [/Niagara/Modules/Spawn/Initialization/InitializeParticle(
    Lifetime=5
    Module.Write Color=true
    Module.Write Lifetime=true
    Module.Write Mass=true
    Module.Write Material Random=true
    Module.Write Position=false
    Module.Write Scale=false
    Module.Write SpriteRotation=false
    Module.Write SpriteSize=true
    Position=&SimulationPosition()
    Sprite Size=<1000 1000>
   )]
  particleUpdateScript [
   &ParticleState()
   /SetVariables(Particles.Debug:NiagaraFloat=&MakeFloatFromVector(Vector=$Engine.Owner.Velocity, Channel=::NewEnumerator0)) {isSetParameters=true, module=SetVariables_#3}
   &DynamicMaterialParameters(
    Index 0 Param 1=$Particles.Debug
    Module.Param0WriteEnabled=true
    Module.Param1WriteEnabled=true
    Module.Param2WriteEnabled=true
    Module.Param3WriteEnabled=true
   )
  ]
  rendererClasses [&NiagaraSpriteRendererProperties]
  renderers [{rendererClass=&NiagaraSpriteRendererProperties}]
 }
 FN_Debug001 {
  bEnabled=false
  emitterUpdateScript [
   &EmitterState()
   &SpawnBurst_Instantaneous(Spawn Count=1, Use Spawn Probability=false)
  ]
  particleSpawnScript [&InitializeParticle(
    Lifetime=5
    Module.Write Color=true
    Module.Write Lifetime=true
    Module.Write Mass=true
    Module.Write Material Random=true
    Module.Write Position=false
    Module.Write Scale=false
    Module.Write SpriteRotation=false
    Module.Write SpriteSize=true
    Position=&SimulationPosition()
    Sprite Size=<1000 1000>
   )]
  particleUpdateScript [
   &ParticleState()
   /SetVariables(Particles.Debug:NiagaraFloat=&MakeFloatFromVector(Vector=$Engine.Owner.Velocity, Channel=::NewEnumerator0)) {isSetParameters=true, module=SetVariables_#3}
   &DynamicMaterialParameters(
    Index 0 Param 1=$Particles.Debug
    Module.Param0WriteEnabled=true
    Module.Param1WriteEnabled=true
    Module.Param2WriteEnabled=true
    Module.Param3WriteEnabled=true
   )
  ]
  rendererClasses [&NiagaraSpriteRendererProperties]
  renderers [{rendererClass=&NiagaraSpriteRendererProperties}]
 }
 FN_Debug002 {
  bEnabled=false
  emitterUpdateScript [
   &EmitterState()
   &SpawnBurst_Instantaneous(Spawn Count=1, Use Spawn Probability=false)
  ]
  particleSpawnScript [&InitializeParticle(
    Lifetime=5
    Module.Write Color=true
    Module.Write Lifetime=true
    Module.Write Mass=true
    Module.Write Material Random=true
    Module.Write Position=false
    Module.Write Scale=false
    Module.Write SpriteRotation=false
    Module.Write SpriteSize=true
    Position=&SimulationPosition()
    Sprite Size=<1000 1000>
   )]
  particleUpdateScript [
   &ParticleState()
   /SetVariables(Particles.Debug:NiagaraFloat=&MakeFloatFromVector(Vector=$Engine.Owner.Velocity, Channel=::NewEnumerator0)) {isSetParameters=true, module=SetVariables_#3}
   &DynamicMaterialParameters(
    Index 0 Param 1=$Particles.Debug
    Module.Param0WriteEnabled=true
    Module.Param1WriteEnabled=true
    Module.Param2WriteEnabled=true
    Module.Param3WriteEnabled=true
   )
  ]
  rendererClasses [&NiagaraSpriteRendererProperties]
  renderers [{rendererClass=&NiagaraSpriteRendererProperties}]
 }
 Light {
  emitterUpdateScript [
   &EmitterState(Life Cycle Mode=::NewEnumerator1, Loop Behavior=::NewEnumerator1, Loop Duration Mode=::NewEnumerator1, UseLoopDelay=true)
   &SpawnBurst_Instantaneous(Use Spawn Probability=false, UseLoopCountLimit=false)
  ]
  particleSpawnScript [
   /Niagara/Modules/Spawn/Initialization/V2/InitializeParticle(
    Color=<rgba 0.3667 0 1 1>
    Lifetime=0.25
    Mesh Scale Mode=::NewEnumerator3
    Mesh Uniform Scale=0.6
    Sprite Size Mode=::NewEnumerator3
    Uniform Sprite Size=920
    UsePositionOffset=false
    Write Material Random=true
   )
   &CameraOffset(Camera Offset Amount=50)
   &InitialMeshOrientation(Mesh Orientation Mode=::NewEnumerator1, Use Rotation Vector=false)
  ]
  particleUpdateScript [
   /SetVariables(Particles.LightExponent=8) {isSetParameters=true}
   &ParticleState()
   &ScaleSpriteSize(
    Uniform Curve Scale=2
    Uniform Curve Sprite Scale=DataInterface() {propertyValues {Curve {keys [
        {arriveTangent=-3.6138, interpMode=RCIM_Cubic, leaveTangent=-3.6138, tangentMode=RCTM_User, time=0.3, value=1}
        {interpMode=RCIM_Cubic, tangentMode=RCTM_User, time=1, value=0}
       ]}}}
   ) {module=ScaleSpriteSize002}
   &SpriteRotationRate(Rotation Rate=-40)
   &ScaleColor(
    Scale Alpha=&FloatFromCurve(FloatCurve=DataInterface() {propertyValues {Curve {keys [<at 0 0 RCIM_Cubic>,<at 0.2 2 RCIM_Cubic>,<at 1 0>]}}})
    Scale RGB=<2 2 2>
    ScaleA=true
    ScaleRGB=true
   )
   /SetVariables(Particles.SpriteSizeDarkGlow:Vector2f=&Multiply_Vector2D_ByFloat(Float=0.2, Vector2D=$Particles.SpriteSize)) {isSetParameters=true}
   /SetVariables(Particles.ColorDarkGlow:LinearColor=&ColorFromCurve(ColorCurve=DataInterface() {propertyValues {
       AlphaCurve {keys [<at 0 0 RCIM_Cubic>,<at 0.133 2.5>,<at 1 0 RCIM_Cubic>]}
       BlueCurve {keys [<at 0 0.01 RCIM_Cubic>]}
       GreenCurve {keys [<at 0 0.008 RCIM_Cubic>]}
       RedCurve {keys [<at 0 0.0087 RCIM_Cubic>]}
      }})) {isSetParameters=true}
   /SetVariables(
    Particles.DynamicMaterialParameterDarkGlow:Vector4f=&MakeVector4(
     X=&Lerp_Float(
      A=1
      B=-0.5
      Alpha=&FloatFromCurve(FloatCurve=DataInterface() {propertyValues {Curve {keys [
           {arriveTangent=1.4226, interpMode=RCIM_Cubic, leaveTangent=1.4226, tangentMode=RCTM_User, time=0, value=0}
           {interpMode=RCIM_Cubic, tangentMode=RCTM_User, time=0.8, value=1}
          ]}}})
     )
     Y=1
     Z=&Lerp_Float(
      A=-2
      B=-1
      Alpha=&FloatFromCurve(FloatCurve=DataInterface() {propertyValues {Curve {keys [<at 0 0>,<at 1 1>]}}})
     )
     W=0.54
    )
    Particles.DynamicMaterialParameterVortex:Vector4f=&MakeVector4(X=0.1, Y=1, Z=0, W=0.25)
   ) {isSetParameters=true}
   /SetVariables(Particles.LightColor=&ColorFromCurve(ColorCurve=DataInterface() {propertyValues {
       AlphaCurve {keys [<at 0.3533 1 RCIM_Cubic>,<at 1 0 RCIM_Cubic>]}
       BlueCurve {keys [<at 0 10 RCIM_Cubic>]}
       GreenCurve {keys [<at 0 0 RCIM_Cubic>]}
       RedCurve {keys [<at 0 3.3333 RCIM_Cubic>]}
      }})) {isSetParameters=true}
  ]
  rendererClasses [&NiagaraLightRendererProperties, &NiagaraSpriteRendererProperties]
  renderers [
   {rendererClass=&NiagaraLightRendererProperties, rendererIndex=0}
   {rendererClass=&NiagaraSpriteRendererProperties, rendererIndex=1}
   {rendererClass=&NiagaraSpriteRendererProperties, rendererIndex=2}
   {rendererClass=&NiagaraSpriteRendererProperties, rendererIndex=3}
  ]
 }
 SmokeWhisps {
  emitterSpawnScript [/SetVariables() {isSetParameters=true}]
  emitterUpdateScript [
   &EmitterState(Life Cycle Mode=::NewEnumerator1, Loop Behavior=::NewEnumerator1, Loop Delay=0.05, Loop Duration Mode=::NewEnumerator1, UseLoopDelay=false)
   &SpawnBurst_Instantaneous(Spawn Count=5, Use Spawn Probability=false, UseLoopCountLimit=false)
   &SpawnRate(
    SpawnRate=&Multiply_Float(
     A=10
     B=&Lerp_Float(
      A=0
      B=1
      Alpha=&FloatFromCurve(
       FloatCurve=DataInterface() {propertyValues {Curve {keys [{interpMode=RCIM_Cubic, tangentMode=RCTM_User, time=0.1, value=1}, {interpMode=RCIM_Cubic, tangentMode=RCTM_User, time=0.5, value=0}]}}}
       CurveIndex=$System.Timeline.NormalizedTime
      )
     )
    )
    Use Spawn Probability=false
   )
  ]
  particleSpawnScript [
   &InitializeParticle(
    Color Maximum=<rgba 0.5664 0.2284 1 1>
    Color Minimum=<rgba 0.2 0.0293 0.0581 1>
    Color Mode=::NewEnumerator2
    Lifetime=&Multiply_Float(A=&RandomRangeFloat(Minimum=0.6, Maximum=1.3), B=1)
    Mass Mode=::NewEnumerator3
    Sprite Rotation Mode=::NewEnumerator2
    Sprite Size Mode=::NewEnumerator1
    Uniform Sprite Size Max=1200
    Uniform Sprite Size Min=760
    UsePositionOffset=false
    Write Material Random=true
   )
   &ShapeLocation(MaskSpawnGroup=false, Offset=<0 -15 0>, Offset Mode=::NewEnumerator0, Override Seed=false, Sphere Radius=9, UseOwnerScale=true)
  ]
  particleUpdateScript [
   &ParticleState()
   &ScaleSpriteSize(
    Scale Sprite Size Mode=::NewEnumerator0
    Uniform Scale Factor=&Multiply_Float(
     A=&Lerp_Float(
      A=0.5
      B=1
      Alpha=&FloatFromCurve(FloatCurve=DataInterface() {propertyValues {Curve {keys [
           {arriveTangent=0.5115, interpMode=RCIM_Cubic, leaveTangent=0.5115, tangentMode=RCTM_User, time=0, value=0}
           {arriveTangent=2.8618, interpMode=RCIM_Cubic, leaveTangent=2.8618, tangentMode=RCTM_User, time=0.7, value=1}
          ]}}})
     )
     B=&FloatFromCurve(
      FloatCurve=DataInterface() {propertyValues {Curve {keys [{interpMode=RCIM_Cubic, tangentMode=RCTM_User, time=0, value=0.3}, {interpMode=RCIM_Cubic, tangentMode=RCTM_User, time=0.25, value=1}]}}}
      CurveIndex=$Emitter.Age
     )
    )
   )
   &ScaleColor(
    Scale Alpha=&Multiply_Float(
     A=&FloatFromCurve(FloatCurve=DataInterface() {propertyValues {Curve {keys [<at 0 0 RCIM_Cubic>,<at 0.2 1 RCIM_Cubic>,<at 1 0 RCIM_Cubic>]}}}, Scale Curve=7)
     B=&FloatFromCurve(
      FloatCurve=DataInterface() {propertyValues {Curve {keys [{interpMode=RCIM_Cubic, tangentMode=RCTM_User, time=0, value=0}, {interpMode=RCIM_Cubic, tangentMode=RCTM_User, time=0.1, value=1}]}}}
      CurveIndex=$Emitter.Age
     )
    )
    Scale RGB=&MakeVectorFromLinearColorRGB(Color=&ColorFromCurve(ColorCurve=DataInterface() {propertyValues {
        AlphaCurve {keys [<at 0 1 RCIM_Cubic>]}
        BlueCurve {keys [<at 0 4>, {arriveTangent=-7.1451, interpMode=RCIM_Cubic, time=0.1569, value=1}, <at 0.5458 0.1>, <at 1 0 RCIM_Cubic>]}
        GreenCurve {keys [<at 0 0.04>, {arriveTangent=-0.0183, interpMode=RCIM_Cubic, time=0.1569, value=0.04}, <at 0.5458 0.03>, <at 1 0 RCIM_Cubic>]}
        RedCurve {keys [<at 0 2.02>, {arriveTangent=-3.6458, interpMode=RCIM_Cubic, time=0.1569, value=0.408}, <at 0.5458 0.03>, <at 1 0 RCIM_Cubic>]}
       }}))
    ScaleA=true
    ScaleRGB=true
   ) {enabled=false}
   &GravityForce(Gravity=<0 0 -200>)
   &Drag(Drag=5, Use Linear Drag=true, Use Rotational Drag=true)
   &InheritSourceMovement(
    Applied Position Scale=&VectorFromFloat(Value=&FloatFromCurve(
      FloatCurve=DataInterface() {propertyValues {Curve {keys [
          {arriveTangent=-3.5756, interpMode=RCIM_Cubic, leaveTangent=-3.5756, tangentMode=RCTM_User, time=0, value=1}
          {interpMode=RCIM_Cubic, tangentMode=RCTM_User, time=0.5, value=0}
         ]}}}
      Scale Curve=0.3
     ))
    ApplyForce=false
    ApplyPosition=true
    ApplyPositionFalloff=false
   )
   &SolveForcesAndVelocity(Clamp Velocity=false, Limit Acceleration=false)
   &ScaleSpriteSizeBySpeed(Max Scale Factor=<100 1>) {enabled=false}
   &SubUVAnimation(
    End Frame Range Override=32
    "Sprite Renderer "=DataInterface() {propertyValues {SpriteRenderer=&self:SmokeWhisps_0.NiagaraSpriteRendererProperties_0}}
    UseEndFrame=true
    UseStartFrame=true
   )
   &DynamicMaterialParameters(
    Index 0 Param 1=&FloatFromCurve(FloatCurve=DataInterface() {propertyValues {Curve {keys [
         {arriveTangent=0.4272, interpMode=RCIM_Cubic, leaveTangent=0.4272, tangentMode=RCTM_User, time=0, value=0.25}
         {arriveTangent=1.2659, interpMode=RCIM_Cubic, leaveTangent=1.2659, tangentMode=RCTM_User, time=1, value=1}
        ]}}})
    Index 0 Param 2=40
    Module.Index0Param0Write=true
    Module.Index0Param1Write=true
    Module.Index0Param2Write=false
    Module.Index0Param3Write=false
   )
   &CameraOffset(Camera Offset Amount=-100)
   &ScaleColor(
    Scale Alpha=&FloatFromCurve(FloatCurve=DataInterface() {propertyValues {Curve {keys [<at 0 1>,<at 1 0>]}}})
    ScaleA=true
    ScaleRGB=true
   ) {module=ScaleColor001}
  ]
  rendererClasses [&NiagaraSpriteRendererProperties]
  renderers [{rendererClass=&NiagaraSpriteRendererProperties}]
 }
 Trail {
  emitterSpawnScript [/SetVariables() {isSetParameters=true}]
  emitterUpdateScript [
   &EmitterState()
   &SpawnPerUnit(
    Max Movement Threshold=1000
    Module.Use Max Movement Threshold=false
    Module.Use Movement Tolerance=false
    Spawn Spacing=100
    Use Spawn Probability=false
   ) {enabled=false}
   &SpawnRate(
    SpawnRate=&Multiply_Float(
     A=55
     B=&FloatFromCurve(
      FloatCurve=DataInterface() {propertyValues {Curve {keys [{interpMode=RCIM_Cubic, tangentMode=RCTM_User, time=0, value=1}, {interpMode=RCIM_Cubic, tangentMode=RCTM_User, time=1, value=0}]}}}
      CurveIndex=$System.Timeline.NormalizedTime
     )
    )
    Use Spawn Probability=false
   )
  ]
  particleSpawnScript [
   &InitializeParticle(
    Color=<rgba 0.2256 0.0142 0.276 1>
    Mass Mode=::NewEnumerator3
    Ribbon Width=800
    Ribbon Width Mode=::NewEnumerator1
    UsePositionOffset=true
    Write Material Random=true
   )
   &CameraOffset(Camera Offset Amount=5000)
   /SetVariables(Particles.PinkColor:LinearColor=<rgba 0.2256 0.0142 0.276 1>) {isSetParameters=true}
   /SetVariables(Particles.PinkWidth=800) {isSetParameters=true}
   /SetVariables(Particles.EnergyColor:LinearColor=<rgba 0.143 0.7276 1 1>, Particles.EnergyWidth=200) {isSetParameters=true}
  ]
  particleUpdateScript [
   &ParticleState()
   &GravityForce(Gravity=<0 0 90>) {enabled=false}
   &ScaleColor(
    Scale Alpha=&Multiply_Float(
     A=&Multiply_Float(
      A=&FloatFromCurve(
       FloatCurve=DataInterface() {propertyValues {Curve {keys [{interpMode=RCIM_Cubic, tangentMode=RCTM_User, time=0, value=1}, {interpMode=RCIM_Cubic, tangentMode=RCTM_User, time=1, value=0}]}}}
       CurveIndex=$System.Timeline.NormalizedTime
      )
      B=&FloatFromCurve(
       FloatCurve=DataInterface() {propertyValues {Curve {keys [{interpMode=RCIM_Cubic, tangentMode=RCTM_User, time=0, value=0}, {interpMode=RCIM_Cubic, tangentMode=RCTM_User, time=0.5, value=1}]}}}
       CurveIndex=&ReturnNormalizedExecIndex()
      )
     )
     B=&Multiply_Float(
      A=&FloatFromCurve(FloatCurve=DataInterface() {propertyValues {Curve {keys [{interpMode=RCIM_Cubic, tangentMode=RCTM_User, time=0, value=0}, {interpMode=RCIM_Cubic, tangentMode=RCTM_User, time=0.07, value=1}]}}})
      B=&FloatFromCurve(
       FloatCurve=DataInterface() {propertyValues {Curve {keys [{interpMode=RCIM_Cubic, tangentMode=RCTM_User, time=0, value=0}, {interpMode=RCIM_Cubic, tangentMode=RCTM_User, time=0.5, value=1}]}}}
       CurveIndex=&ReturnNormalizedExecIndex()
      )
     )
    )
    ScaleA=true
    ScaleRGB=false
   ) {module=ScaleColor001}
   &ScaleColor(
    Scale Alpha=&FloatFromCurve(FloatCurve=DataInterface() {propertyValues {Curve {
        defaultValue=1
        keys [{interpMode=RCIM_Cubic, tangentMode=RCTM_User, time=0, value=0}, {interpMode=RCIM_Cubic, tangentMode=RCTM_User, time=0.3, value=1}]
       }}})
    ScaleA=true
    ScaleRGB=true
   )
   &SolveForcesAndVelocity(Clamp Velocity=false, Limit Acceleration=false)
   &ScaleRibbonWidth(Ribbon Width Scale=&FloatFromCurve(FloatCurve=DataInterface() {propertyValues {Curve {keys [
         {arriveTangent=19.2206, interpMode=RCIM_Cubic, leaveTangent=19.2206, tangentMode=RCTM_User, time=0, value=0}
         {arriveTangent=2.9167, interpMode=RCIM_Cubic, leaveTangent=2.9167, time=0.1, value=1.6}
         {arriveTangent=-1, interpMode=RCIM_Cubic, leaveTangent=-1, time=0.3, value=0.875}
         {arriveTangent=0.3186, interpMode=RCIM_Cubic, leaveTangent=0.3186, tangentMode=RCTM_User, time=0.7, value=1}
        ]}}}))
   &DynamicMaterialParameters(
    Index 0 Param 1=&FloatFromCurve(FloatCurve=DataInterface() {propertyValues {Curve {keys [
         {interpMode=RCIM_Cubic, leaveTangent=-9.1594, tangentMode=RCTM_Break, time=0, value=1}
         {interpMode=RCIM_Cubic, leaveTangent=-5.9384, tangentMode=RCTM_Break, time=0.3, value=0}
        ]}}})
    Index 0 Param 2=&FloatFromCurve(FloatCurve=DataInterface() {propertyValues {Curve {keys [{tangentMode=RCTM_User, time=0, value=0}, {tangentMode=RCTM_User, time=1, value=1}]}}})
    Index 0 Param 4=&FloatFromCurve(FloatCurve=DataInterface() {propertyValues {Curve {keys [{interpMode=RCIM_Cubic, leaveTangent=-13.8479, tangentMode=RCTM_Break, time=0, value=3}, <at 0.5 0 RCIM_Cubic>]}}})
    Module.Index0Param0Write=true
    Module.Index0Param1Write=true
    Module.Index0Param2Write=true
    Module.Index0Param3Write=true
   )
   &Color(Color=&ColorFromCurve(ColorCurve=DataInterface() {propertyValues {
       AlphaCurve {keys [<at 0 1 RCIM_Cubic>,<at 1 0 RCIM_Cubic>]}
       BlueCurve {keys [<at 0 0.1 RCIM_Cubic>,<at 0.1197 1 RCIM_Cubic>]}
       GreenCurve {keys [<at 0 0.0592 RCIM_Cubic>,<at 0.1197 0.0008 RCIM_Cubic>]}
       RedCurve {keys [<at 0 0 RCIM_Cubic>,<at 0.1197 0.2804 RCIM_Cubic>]}
      }})) {enabled=false}
   /SetVariables(Particles.PinkColor:LinearColor=&MultiplyLinearColor(
     Linear Color=$Particles.Initial.PinkColor
     Scale Factor=&MakeVector4(
      X=1
      Y=1
      Z=1
      W=&FloatFromCurve(FloatCurve=DataInterface() {propertyValues {Curve {keys [<at 0 1>,<at 0.01 0>]}}})
     )
    )) {isSetParameters=true}
   /SetVariables(
    Particles.EnergyColor:LinearColor=&MultiplyLinearColor(
     Linear Color=$Particles.Initial.EnergyColor
     Scale Factor=&MakeVector4(
      X=1
      Y=1
      Z=1
      W=&FloatFromCurve(FloatCurve=DataInterface() {propertyValues {Curve {keys [<at 0.03 0>,<at 0.05 0.7>,<at 1 0>]}}})
     )
    )
    Particles.EnergyDyn:Vector4f=&MakeVector4(X=0, Y=0, Z=1, W=1)
    Particles.EnergyWidth:NiagaraFloat=&Multiply_Float(
     A=$Particles.Initial.EnergyWidth
     B=&FloatFromCurve(FloatCurve=DataInterface() {propertyValues {Curve {keys [
          {arriveTangent=19.2206, interpMode=RCIM_Cubic, leaveTangent=19.2206, tangentMode=RCTM_User, time=0, value=0}
          {arriveTangent=2.9167, interpMode=RCIM_Cubic, leaveTangent=2.9167, time=0.1, value=1.3}
          {arriveTangent=-2.1667, interpMode=RCIM_Cubic, leaveTangent=-2.1667, time=0.3, value=0.875}
          {arriveTangent=0.3186, interpMode=RCIM_Cubic, leaveTangent=0.3186, tangentMode=RCTM_User, time=0.7, value=0}
         ]}}})
    )
    Particles.PinkDyn:Vector4f=&MakeVector4(
     X=0.5
     Y=0
     Z=0
     W=&FloatFromCurve(FloatCurve=DataInterface() {propertyValues {Curve {keys [<at 0 1>,<at 1 1>]}}})
    )
    Particles.PinkWidth:NiagaraFloat=&Multiply_Float(
     A=$Particles.Initial.PinkWidth
     B=&FloatFromCurve(FloatCurve=DataInterface() {propertyValues {Curve {keys [
          {arriveTangent=19.2206, interpMode=RCIM_Cubic, leaveTangent=19.2206, tangentMode=RCTM_User, time=0, value=0}
          {arriveTangent=2.0833, interpMode=RCIM_Cubic, leaveTangent=2.0833, time=0.1, value=1}
          {arriveTangent=-1.6667, interpMode=RCIM_Cubic, leaveTangent=-1.6667, time=0.3, value=0.625}
          {arriveTangent=0.3186, interpMode=RCIM_Cubic, leaveTangent=0.3186, tangentMode=RCTM_User, time=0.7, value=0}
         ]}}})
    )
   ) {isSetParameters=true}
  ]
  rendererClasses [&NiagaraRibbonRendererProperties]
  renderers [
   {rendererClass=&NiagaraRibbonRendererProperties}
   {rendererClass=&NiagaraRibbonRendererProperties, rendererIndex=1}
   {rendererClass=&NiagaraRibbonRendererProperties, rendererIndex=2}
  ]
 }
 VoidOrb {
  emitterUpdateScript [
   &EmitterState(
    Inactive Response=::NewEnumerator1
    Life Cycle Mode=::NewEnumerator1
    Loop Behavior=::NewEnumerator1
    Loop Delay=0.5
    Loop Duration Mode=::NewEnumerator1
    UseLoopDelay=false
   )
   &SpawnBurst_Instantaneous(Spawn Count=4, Use Spawn Probability=false, UseLoopCountLimit=false)
   &SpawnRate(SpawnRate=7, Use Spawn Probability=false) {enabled=false}
   &SpawnRate(SpawnGroup=1, SpawnRate=7, Use Spawn Probability=false) {module=SpawnRate001}
   /SetVariables(
    Emitter.CameraOfsCore=10
    Emitter.CameraOfs_Outline=-10
    Emitter.Color:LinearColor=&ColorFromCurve(
     ColorCurve=DataInterface() {propertyValues {
       AlphaCurve {keys [<at 0.096 0 RCIM_Cubic>,<at 0.3023 1 RCIM_Cubic>]}
       BlueCurve {keys [<at 0 5 RCIM_Cubic>]}
       GreenCurve {keys [<at 0 0 RCIM_Cubic>]}
       RedCurve {keys [<at 0 1.6667 RCIM_Cubic>]}
      }}
     CurveIndex=$Emitter.Age
    )
    Emitter.Color_Core:LinearColor=<rgba 0.441 0.0208 1 1>
    Emitter.Color_Outline:LinearColor=<rgba 1 0.4115 0.0283 1>
   ) {isSetParameters=true}
   /SetVariables(Emitter.Scale:Vector3f=&VectorFromCurve(
     VectorCurve=DataInterface() {propertyValues {
       XCurve {keys [{tangentMode=RCTM_User, time=0, value=0.1}, {tangentMode=RCTM_User, time=0.2, value=1}]}
       YCurve {keys [{tangentMode=RCTM_User, time=0, value=0.1}, {tangentMode=RCTM_User, time=0.2, value=1}]}
       ZCurve {keys [{tangentMode=RCTM_User, time=0, value=0.1}, {tangentMode=RCTM_User, time=0.2, value=1}]}
      }}
     CurveIndex=$Emitter.Age
     Scale Curve=&VectorFromFloat(Value=&Multiply_Float(A=5.2, B=$System.FadeOutFastTimeline))
    )) {isSetParameters=true}
   /SetVariables(Emitter.Scale2:Vector3f=&Multiply_Vector(
     A=$Emitter.Scale
     B=&VectorFromFloat(Value=&Multiply_Float(A=1.6, B=$System.FadeOutFastTimeline))
    )) {isSetParameters=true}
   &SpawnBurst_Instantaneous(Spawn Time=8.55, Use Spawn Probability=false, UseLoopCountLimit=false) {enabled=false, module=SpawnBurst_Instantaneous001}
  ]
  particleSpawnScript [
   &InitializeParticle(
    Color=&SelectLinearColorFromArray(
     Array Sampling Mode=::NewEnumerator0
     Direct Array Index=?
     Color Selection Array=DataInterface() {propertyValues {ColorData [<rgba 2.5316 0.4289 3 1>, <rgba 0.0507 0.7272 3 2>]}}
     Random Seed=0
    )
    Lifetime=15
    Mesh Scale Mode=::NewEnumerator3
    Mesh Uniform Scale=&MakeFloatFromVector(
     Vector=&VectorFromCurve(
      VectorCurve=DataInterface() {propertyValues {
        XCurve {keys [{tangentMode=RCTM_User, time=0, value=1}, {tangentMode=RCTM_User, time=0.2, value=1}]}
        YCurve {keys [{tangentMode=RCTM_User, time=0, value=1}, {tangentMode=RCTM_User, time=0.2, value=1}]}
        ZCurve {keys [{tangentMode=RCTM_User, time=0, value=1}, {tangentMode=RCTM_User, time=0.2, value=1}]}
       }}
      CurveIndex=$Emitter.Age
      Scale Curve=&VectorFromFloat(Value=&Multiply_Float(A=5.2, B=$System.FadeOutFastTimeline))
     )
     Channel=::NewEnumerator0
    )
    Sprite Rotation Mode=::NewEnumerator2
    Sprite Size=&Vector2DFromFloat(Value=&Multiply_Float(
      A=&MaskFloatBySpawnGroup(
       Masked Float=&RandomRangeFloat(Minimum=640, Maximum=760)
       Passthrough Float=&RandomRangeFloat(Minimum=560, Maximum=640)
      )
      B=0.9
     ))
    Sprite Size Mode=::NewEnumerator4
    Sprite UV Mode=::NewEnumerator4
    Sprite UV Scale:Vector2f=<-1 1>
    UsePositionOffset=false
    Write Material Random=true
   )
   &CameraOffset(Camera Offset Amount=-15)
   &DynamicMaterialParameters(
    Index 0 Param 1=&RandomRangeFloat(Minimum=0, Maximum=-6)
    Module.Index0Param0Write=true
    Module.Index0Param1Write=false
    Module.Index0Param2Write=false
    Module.Index0Param3Write=false
   )
   &AddVelocity(Velocity=&MaskVectorBySpawnGroup(Masked Vector=&RandomRangeVector(Minimum=<0 0 30>, Maximum=<0 0 -30>), Passthrough Vector=<0 0 0>)) {enabled=false}
   &AddVelocity(Velocity=$System.DebugSpeed) {module=AddVelocity001}
   /SetVariables(Particles.VisibilityTag=&SelectIntFromArray(Direct Array Index=&SpawnGroup(), Int Selection Array=DataInterface() {propertyValues {IntData [0,1]}})) {isSetParameters=true}
  ]
  particleUpdateScript [
   &ParticleState()
   /SetVariables(Particles.SpriteAge:NiagaraFloat=&FloatFromCurve(
     FloatCurve=DataInterface() {propertyValues {Curve {defaultValue=1, keys [{tangentMode=RCTM_User, time=0, value=0}, {tangentMode=RCTM_User, time=1, value=1}]}}}
     CurveIndex=$Particles.Age
    )) {isSetParameters=true}
   &ScaleSpriteSize(
    Scale Sprite Size Mode=::NewEnumerator0
    Uniform Scale Factor=&FloatFromCurve(
     FloatCurve=DataInterface() {propertyValues {Curve {keys [<at 0 0.1 RCIM_Cubic>, {interpMode=RCIM_Cubic, leaveTangent=0.1472, time=0.4, value=1}]}}}
     CurveIndex=$Particles.SpriteAge
    )
   )
   &ScaleColor(
    Scale Alpha=&FloatFromCurve(
     FloatCurve=DataInterface() {propertyValues {Curve {keys [
         {arriveTangent=9.8672, interpMode=RCIM_Cubic, leaveTangent=9.8672, tangentMode=RCTM_User, time=0, value=0}
         <at 0.2 1 RCIM_Cubic>
         {interpMode=RCIM_Cubic, leaveTangent=-26.9591, time=1, value=0}
        ]}}}
     Scale Curve=$System.FadeOutFastTimeline
    )
    Scale RGB=&VectorFromCurve(VectorCurve=DataInterface() {propertyValues {
       XCurve {keys [<at 0.25 1 RCIM_Cubic>,<at 1 0.5>]}
       YCurve {keys [<at 0.25 1 RCIM_Cubic>,<at 1 0.5>]}
       ZCurve {keys [<at 0.25 1 RCIM_Cubic>,<at 1 0.5>]}
      }})
    ScaleA=true
    ScaleRGB=true
   )
   /Niagara/Modules/Update/SubUV/SubUVAnimation(Number Of Frames=32, SubUV Lookup Index=$Particles.SpriteAge)
   &DynamicMaterialParameters(
    Index 0 Param 1=&FloatFromCurve(
     FloatCurve=DataInterface() {propertyValues {Curve {keys [
         {arriveTangent=2.8718, interpMode=RCIM_Cubic, leaveTangent=2.8718, tangentMode=RCTM_User, time=0, value=0}
         {arriveTangent=-0.3942, interpMode=RCIM_Cubic, leaveTangent=-0.3942, tangentMode=RCTM_User, time=1, value=1}
        ]}}}
     Scale Curve=0.1
    )
    Index 0 Param 2=50
    Module.Index0Param0Write=true
    Module.Index0Param1Write=true
    Module.Index0Param2Write=false
    Module.Index0Param3Write=false
   ) {module=DynamicMaterialParameters001}
   &Drag(Drag=2, Use Linear Drag=true, Use Rotational Drag=true)
   &GravityForce(Gravity=<0 0 -100>) {enabled=false}
   &InheritSourceMovement(
    Applied Position Scale=&VectorFromFloat(Value=&FloatFromCurve(FloatCurve=DataInterface() {propertyValues {Curve {keys [
          {interpMode=RCIM_Cubic, tangentMode=RCTM_User, time=0.35, value=1}
          {arriveTangent=-2.5788, interpMode=RCIM_Cubic, leaveTangent=-2.5788, tangentMode=RCTM_User, time=1, value=0}
         ]}}}))
    ApplyForce=false
    ApplyPosition=true
    ApplyPositionFalloff=false
   ) {enabled=false}
   &SolveForcesAndVelocity(Clamp Velocity=false, Limit Acceleration=false)
   &ScaleMeshSizeBySpeed(Max Scale Factor=<1.35 1 1>, Source Velocity:Vector3f=$Engine.Owner.Velocity, Speed Threshold=1000)
   &UpdateMeshOrientation(
    Facing Direction=$Engine.Owner.Velocity
    Facing Interpolation=false
    Orientation Method=::NewEnumerator1
    Override Quat=false
    TransformFacing=false
    TransformUp=false
    Up Direction:Vector3f=<1 0 0>
   )
   /SetVariables(Particles.Position=&SmoothLerpOverTime_Position(
     Convergence Rate:NiagaraFloat=4.5
     Smooth Value:NiagaraPosition=$Particles.Presolve.Position
     Target Value:NiagaraPosition=$Engine.Owner.Position
    )) {isSetParameters=true}
   /SetVariables(Particles.OrbColor:LinearColor=&ColorFromCurve(
     ColorCurve=DataInterface() {propertyValues {
       AlphaCurve {keys [<at 0.096 0 RCIM_Cubic>,<at 0.3023 1 RCIM_Cubic>]}
       BlueCurve {keys [<at 0 5 RCIM_Cubic>]}
       GreenCurve {keys [<at 0 0 RCIM_Cubic>]}
       RedCurve {keys [<at 0 1.6667 RCIM_Cubic>]}
      }}
     CurveIndex=$Emitter.Age
    )) {isSetParameters=true}
   &ScaleColor(
    Scale Alpha=&FloatFromCurve(FloatCurve=DataInterface() {propertyValues {Curve {
        defaultValue=1
        keys [
         {interpMode=RCIM_Cubic, tangentMode=RCTM_User, time=0, value=0}
         {interpMode=RCIM_Cubic, tangentMode=RCTM_User, time=0.02, value=1}
         {tangentMode=RCTM_User, time=0.05, value=0}
        ]
       }}})
    ScaleA=true
    ScaleRGB=true
   ) {module=ScaleColor001}
   &KillParticles(Kill Particles=&Boolean_AndOperation(
     Bool1=&SetBoolByFloatComparison(A=$Particles.SpriteAge, B=0.99)
     Bool2=&SetBoolByIntComparison(A=$Particles.VisibilityTag, B=1, Comparison Type=::NewEnumerator4)
    ))
  ]
  rendererClasses [&NiagaraSpriteRendererProperties, &NiagaraMeshRendererProperties]
  renderers [
   {rendererClass=&NiagaraSpriteRendererProperties}
   {rendererClass=&NiagaraMeshRendererProperties, rendererIndex=1}
  ]
 }
}
summary {
 systemName=NS_TentaclePortalProjectile
 emitters [
  {emitterName=Light, rendererClasses [&NiagaraLightRendererProperties, &NiagaraSpriteRendererProperties]}
  {emitterName=VoidOrb, rendererClasses [&NiagaraSpriteRendererProperties, &NiagaraMeshRendererProperties]}
  {emitterName=ChromaticRefraction, rendererClasses [&NiagaraSpriteRendererProperties]}
  {emitterName=SmokeWhisps, rendererClasses [&NiagaraSpriteRendererProperties]}
  {emitterName=Trail, rendererClasses [&NiagaraRibbonRendererProperties]}
  {bEnabled=false, emitterName=FN_Debug, rendererClasses [&NiagaraSpriteRendererProperties]}
  {bEnabled=false, emitterName=FN_Debug001, rendererClasses [&NiagaraSpriteRendererProperties]}
  {bEnabled=false, emitterName=FN_Debug002, rendererClasses [&NiagaraSpriteRendererProperties]}
 ]
 userVariables [User.ProjStopped {type=NiagaraBool}]
}
